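# Control Rig for the F_MED_Hacker female player body that adds corrective deformation to the character and outfit bones.
# After the shared DeformRig function, elbow yaw deltas from reference pose rotate sleeve bones and write inverse-bend curves.
# Those curves weight sleeve offsets, branched on isLoFi, with LOD-scaled shifts when true.
# Neck yaw and spherical pose readers drive collar and foot correctives, and is DBNO rotates the hoodie back.
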ControlRigBlueprint /BRCosmetics/Characters/Player/Female/Medium/Bodies/F_MED_Hacker/Meshes/F_MED_Hacker_CtrlRig
graphs [
 ContainedGraph {nodes [
   n1 = RigVMFunctionEntryNode()
   n2 = &RigUnit_GetTransform(Space=LocalSpace, Item=@n1.Item)
   n3 = &RigUnit_GetTransform(Space=LocalSpace, bInitial=true, Item=@n1.Item)
   > RigVMFunctionReturnNode(OutTransform=/Script/RigVM.RigVMFunction_MathTransformToEulerTransform(Value=&RigVMFunction_MathTransformMakeRelative(Global=@n2.Transform, Parent=@n3.Transform).Local) {is=struct RigVMUnitNode, template="Cast::Execute(in Value,out Result)"}.Result)
  ]}
 RigVMFunctionLibrary {nodes [DeformRig_GetDeltaRefTransform_FNC = RigVMCollapseNode(ExecuteContext=(), Item=(Type=Bone))]}
 RigVMFunction_Sequence_2_ContainedGraph {nodes [
   n1 = RigVMFunctionEntryNode()
   RigVMFunction_Sequence = &RigVMFunction_Sequence(ExecuteContext=@n1.ExecuteContext)
   n2 = &RigVMFunction_Sequence(ExecuteContext=@RigVMFunction_Sequence.B)
   n3 = &RigVMFunction_Sequence(ExecuteContext=@n2.B)
   n4 = &RigVMFunction_Sequence(ExecuteContext=@n3.B)
   RigVMFunctionReturnNode(A=@RigVMFunction_Sequence.A, B=@n2.A, C=@n3.A, D=@n4.A, E=@n4.B)
  ]}
 RigVMFunction_Sequence_3_ContainedGraph {nodes [
   n1 = RigVMFunctionEntryNode()
   n2 = &RigVMFunction_Sequence(ExecuteContext=@n1.ExecuteContext)
   n3 = &RigVMFunction_Sequence(ExecuteContext=@n2.B)
   RigVMFunctionReturnNode(A=@n2.A, B=@n3.A, C=@n3.B)
  ]}
 RigVMModel {nodes [
   RigUnit_BeginExecution = &RigUnit_BeginExecution()
   n1 = &RigVMFunction_Sequence(ExecuteContext=@RigUnit_BeginExecution.ExecutePin)
   FN_DeformRig_FNC = RigVMFunctionReferenceNode(
    DeformRig_ Curve="(deformRig_l_upperarm_up_alpha_72_#1=False,deformRig_r_upperarm_up_alpha_73_#2=false,deformRig_l_upperarm_fr_alpha_74_#3=False,deformRig_r_upperarm_fr_alpha_75_CB51B8 ...(truncated)"
    DeformRig_Bool="(upperarm_1_#4=True,elbow_3_#5=True,wrist_7_#6=True,groin_8_#7=True,glute_17_#8=True,knee_9_40C408324CA261 ...(truncated)"
    DeformRig_Debug=(upperarm_debug_18_#4=False,wrist_debug_20_#6=False,lat_debug_26_#9=False,groinGlute_debug_27_#7=False)
    DeformRig_Value="(upperarm_up_18_#10=(Rotation=(X=0,Y=0,Z=0,W=1),Translation=(X=0,Y=0,Z=9),Scale3D=(X=0.7,Y=1,Z=1)),upperarm_fwd_fr_19_#11=(Rotation=(X=0,Y=0,Z=0,W=1),Translation=(X=0,Y=14,Z=0),Scale3D=(X=0.5,Y=1,Z= ...(truncated)"
    ExecuteContext=@n1.B
   )
   > RigVMFunction_Sequence = RigVMAggregateNode()
   RigVMFunction_Sequence_1 = RigVMAggregateNode(ExecuteContext=@RigVMFunction_Sequence.A)
   DeformRig_GetDeltaRefTransform_FNC = RigVMFunctionReferenceNode(Item=(Type=Bone,Name="lowerarm_l"), ExecuteContext=@RigVMFunction_Sequence_1.A)
   n2 = &RigVMFunction_MathFloatRemap(
    SourceMaximum=110
    SourceMinimum=70
    TargetMaximum=5
    TargetMinimum=0
    bClamp=true
    Value=@DeformRig_GetDeltaRefTransform_FNC.OutTransform.Rotation.Yaw
   )
   Multiply = &RigVMFunction_MathDoubleMul(
    B=&RigVMFunction_MathFloatRemap(
     SourceMaximum=110
     SourceMinimum=70
     TargetMaximum=-0.08
     TargetMinimum=-0.18
     bClamp=true
     Value=@DeformRig_GetDeltaRefTransform_FNC.OutTransform.Rotation.Yaw
    ).Result
    A=@DeformRig_GetDeltaRefTransform_FNC.OutTransform.Rotation.Yaw
   )
   Make = RigVMDispatchNode(Elements=(Pitch=5,Yaw=0,Roll=5), Elements.Pitch=@n2.Result, Elements.Roll=@n2.Result, Elements.Yaw=@Multiply.Result) {template="DISPATCH_RigVMDispatch_MakeStruct(in Elements,out Struct)"}
   RigVMFunction_MathQuaternionFromRotatorV2 = /Script/RigVM.RigVMFunction_MathQuaternionFromRotatorV2(Value=@Make.Struct) {is=struct RigVMUnitNode, template="Cast::Execute(in Value,out Result)"}
   > &RigUnit_ModifyTransforms(
    ItemToModify=((Item=(Type=Bone,Name="deform_sleeve_l"),Transform=(Rotation=(X=0,Y=0,Z=0,W=1),Translation=(X=0,Y=0,Z=0),Scale3D=(X=1,Y=1,Z=1))))
    Weight=1
    WeightMinimum=0
    ItemToModify.0.Transform.Rotation=@RigVMFunction_MathQuaternionFromRotatorV2.Result
   )
   > /Script/ControlRig.RigUnit_SetCurveValue(
    Curve=ElbowLBentInv
    Value=&RigVMFunction_MathFloatRemap(
     SourceMaximum=110
     SourceMinimum=90
     TargetMaximum=0
     TargetMinimum=1
     bClamp=true
     Value=@DeformRig_GetDeltaRefTransform_FNC.OutTransform.Rotation.Yaw
    ).Result
   ) {is=struct RigVMUnitNode}
   DeformRig_GetDeltaRefTransform_FNC_1 = RigVMFunctionReferenceNode(Item=(Type=Bone,Name="lowerarm_r"), ExecuteContext=@RigVMFunction_Sequence_1.B)
   n3 = &RigVMFunction_MathFloatRemap(
    SourceMaximum=110
    SourceMinimum=70
    TargetMaximum=5
    TargetMinimum=0
    bClamp=true
    Value=@DeformRig_GetDeltaRefTransform_FNC_1.OutTransform.Rotation.Yaw
   )
   n4 = &RigVMFunction_MathDoubleMul(
    B=&RigVMFunction_MathFloatRemap(
     SourceMaximum=110
     SourceMinimum=70
     TargetMaximum=-0.08
     TargetMinimum=-0.18
     bClamp=true
     Value=@DeformRig_GetDeltaRefTransform_FNC_1.OutTransform.Rotation.Yaw
    ).Result
    A=@DeformRig_GetDeltaRefTransform_FNC_1.OutTransform.Rotation.Yaw
   )
   > &RigUnit_ModifyTransforms(
    ItemToModify=((Item=(Type=Bone,Name="deform_sleeve_r"),Transform=(Rotation=(X=0,Y=0,Z=0,W=1),Translation=(X=0,Y=0,Z=0),Scale3D=(X=1,Y=1,Z=1))))
    Weight=1
    WeightMinimum=0
    ItemToModify.0.Transform.Rotation=&RigVMFunction_MathQuaternionFromRotatorV2(Value=RigVMDispatchNode(Elements=(Pitch=0,Yaw=0,Roll=0), Elements.Pitch=@n3.Result, Elements.Roll=@n3.Result, Elements.Yaw=@n4.Result) {template="DISPATCH_RigVMDispatch_MakeStruct(in Elements,out Struct)"}.Struct) {template="Cast::Execute(in Value,out Result)"}.Result
   )
   > &RigUnit_SetCurveValue(
    Curve=ElbowRBentInv
    Value=&RigVMFunction_MathFloatRemap(
     SourceMaximum=110
     SourceMinimum=90
     TargetMaximum=0
     TargetMinimum=1
     bClamp=true
     Value=@DeformRig_GetDeltaRefTransform_FNC_1.OutTransform.Rotation.Yaw
    ).Result
   )
   n5 = &RigVMFunction_ControlFlowBranch(ExecuteContext=@RigVMFunction_Sequence_1.C, Condition=RigVMVariableNode() {variable=isLoFi}.Value)
   &RigUnit_ModifyTransforms(
    ItemToModify=((Item=(Type=Bone,Name="deform_sleeve_r"),Transform=(Rotation=(X=0,Y=0,Z=0,W=1),Translation=(X=0,Y=-0.5,Z=0),Scale3D=(X=1,Y=1,Z=1))))
    WeightMinimum=0
    Weight=/Script/ControlRig.RigUnit_GetCurveValue(Curve=ElbowRBentInv) {is=struct RigVMUnitNode}.Value
    ExecutePin=@n5.False
   )
   > &RigUnit_ModifyTransforms(
    ItemToModify=((Item=(Type=Bone,Name="deform_sleeve_l"),Transform=(Rotation=(X=0,Y=0,Z=0,W=1),Translation=(X=0,Y=0.5,Z=0),Scale3D=(X=1,Y=1,Z=1))))
    WeightMinimum=0
    Weight=&RigUnit_GetCurveValue(Curve=ElbowLBentInv).Value
   )
   n6 = &RigUnit_GetCurveValue(Curve=ElbowRBentInv)
   RigVMFunction_MathIntToDouble = /Script/RigVM.RigVMFunction_MathIntToFloat(Value=RigVMVariableNode() {variable=LOD}.Value) {is=struct RigVMUnitNode, template="Cast::Execute(in Value,out Result)"}
   n7 = &RigVMFunction_MathDoubleMul(A=3, B=@RigVMFunction_MathIntToDouble.Result)
   &RigUnit_ModifyTransforms(
    ItemToModify=((Item=(Type=Bone,Name="dyn_sleeveBk_r_01"),Transform=(Rotation=(X=0,Y=0,Z=0,W=1),Translation=(X=3,Y=0,Z=0),Scale3D=(X=1.7,Y=1,Z=1))))
    WeightMinimum=0
    Weight=@n6.Value
    ExecutePin=@n5.True
    ItemToModify.0.Transform.Translation.X=@n7.Result
   )
   > &RigUnit_ModifyTransforms(
    ItemToModify="((Item=(Type=Bone,Name=\"deform_sleeve_r\"),Transform=(Rotation=(X=0,Y=0,Z=0,W=1),Translation=(X=0,Y=-1,Z=0),Scale3D=(X=1,Y=1,Z=1))),(Item=(Type=Bone,Name=\"deform_elbow_out_r\"),Transform=(Rotation=(X=0,Y=0,Z=0,W=1),Translation=(X=0,Y=-1.5,Z=0),Scale3D=(X=1,Y ...(truncated)"
    WeightMinimum=0
    Weight=@n6.Value
   )
   > &RigUnit_ModifyTransforms(
    ItemToModify=((Item=(Type=Bone,Name="deform_sleeve_l"),Transform=(Rotation=(X=0,Y=0,Z=0,W=1),Translation=(X=0,Y=1,Z=0),Scale3D=(X=1.5,Y=1,Z=1))))
    WeightMinimum=0
    Weight=&RigUnit_GetCurveValue(Curve=ElbowLBentInv).Value
    ItemToModify.0.Transform.Translation.X=&RigVMFunction_MathDoubleMul(B=-1, A=@n7.Result).Result
   )
   Get Delta Ref = RigVMFunctionReferenceNode(Item=(Type=Bone,Name="neck_02"), ExecuteContext=@RigVMFunction_Sequence.B)
   > &RigUnit_ModifyTransforms(
    ItemToModify=((Item=(Type=Bone,Name="deform_collar_bk"),Transform=(Rotation=(X=0,Y=0,Z=0,W=1),Translation=(X=0,Y=0,Z=0),Scale3D=(X=1,Y=1,Z=1))))
    Weight=1
    WeightMinimum=0
    ItemToModify.0.Transform.Rotation=&RigVMFunction_MathQuaternionFromRotatorV2(Value=RigVMDispatchNode(Elements=(Pitch=0,Yaw=0,Roll=0), Elements.Yaw=@"Get Delta Ref".Result.Rotation.Yaw) {template="DISPATCH_RigVMDispatch_MakeStruct(in Elements,out Struct)"}.Struct) {template="Cast::Execute(in Value,out Result)"}.Result
   )
   > n8 = &RigUnit_SphericalPoseReader(
    ActiveRegionScaleFactors=(PositiveWidth=1,NegativeWidth=1,PositiveHeight=1,NegativeHeight=1)
    ActiveRegionSize=0.1
    Debug=(bDrawDebug=false,bDraw2D=False,bDrawLocalAxes=False,DebugScale=25,DebugSegments=20,DebugThickness=0.25)
    DriverAxis=(X=1,Y=0,Z=0)
    DriverItem=(Type=Bone,Name="clavicle_l")
    FalloffRegionScaleFactors=(PositiveWidth=1,NegativeWidth=1,PositiveHeight=1,NegativeHeight=1)
    FalloffSize=0.2
    FlipHeightScaling=False
    FlipWidthScaling=False
    OptionalParentItem=(Type=Bone,Name="None")
    RotationOffset=(X=45,Y=0,Z=-90)
   )
   > &RigUnit_ModifyTransforms(
    ItemToModify=((Item=(Type=Bone,Name="deform_collar_l"),Transform=(Rotation=(X=0,Y=-0.1737,Z=0,W=0.9848),Translation=(X=-0.3953,Y=0,Z=1.0858),Scale3D=(X=1,Y=1,Z=1))))
    WeightMinimum=0
    Weight=@n8.OutputParam
   )
   > n9 = &RigUnit_SphericalPoseReader(
    ActiveRegionScaleFactors=(PositiveWidth=1,NegativeWidth=1,PositiveHeight=1,NegativeHeight=1)
    ActiveRegionSize=0.1
    Debug=(bDrawDebug=false,bDraw2D=False,bDrawLocalAxes=False,DebugScale=25,DebugSegments=20,DebugThickness=0.25)
    DriverAxis=(X=1,Y=0,Z=0)
    DriverItem=(Type=Bone,Name="clavicle_r")
    FalloffRegionScaleFactors=(PositiveWidth=1,NegativeWidth=1,PositiveHeight=1,NegativeHeight=1)
    FalloffSize=0.2
    FlipHeightScaling=False
    FlipWidthScaling=False
    OptionalParentItem=(Type=Bone,Name="None")
    RotationOffset=(X=45,Y=0,Z=-90)
   )
   > &RigUnit_ModifyTransforms(
    ItemToModify=((Item=(Type=Bone,Name="deform_collar_r"),Transform=(Rotation=(X=0,Y=-0.1736,Z=0,W=0.9848),Translation=(X=-0.39,Y=0,Z=1.085),Scale3D=(X=1,Y=1,Z=1))))
    WeightMinimum=0
    Weight=@n9.OutputParam
   )
   > n10 = &RigUnit_SphericalPoseReader(
    ActiveRegionScaleFactors=(PositiveWidth=1,NegativeWidth=1,PositiveHeight=1,NegativeHeight=1)
    ActiveRegionSize=0.1
    Debug=(bDrawDebug=True,bDraw2D=False,bDrawLocalAxes=False,DebugScale=25,DebugSegments=20,DebugThickness=0.25)
    DriverAxis=(X=1,Y=0,Z=0)
    DriverItem=(Type=Bone,Name="head")
    FalloffRegionScaleFactors=(PositiveWidth=1,NegativeWidth=1,PositiveHeight=1,NegativeHeight=1)
    FalloffSize=0.2
    FlipHeightScaling=False
    FlipWidthScaling=False
    OptionalParentItem=(Type=Bone,Name="spine_05")
    RotationOffset=(X=-16,Y=-18,Z=-10)
   )
   > &RigUnit_ModifyTransforms(
    ItemToModify="((Item=(Type=Bone,Name=\"deform_collar_r\"),Transform=(Rotation=(X=0,Y=-0.0337,Z=0,W=0.9994),Translation=(X=0,Y=0,Z=1.0668),Scale3D=(X=1,Y=1,Z=1))),(Item=(Type=Bone,Name=\"deform_collar_bk\"),Transform=(Rotation=(X=0,Y=0,Z=0,W=1),Translation=(X=0,Y=0,Z=3.5509) ...(truncated)"
    WeightMinimum=0
    Weight=@n10.OutputParam
   )
   n11 = &RigUnit_SphericalPoseReader(
    ActiveRegionScaleFactors=(PositiveWidth=1,NegativeWidth=1,PositiveHeight=1,NegativeHeight=1)
    ActiveRegionSize=0.1
    Debug=(bDrawDebug=false,bDraw2D=False,bDrawLocalAxes=False,DebugScale=25,DebugSegments=20,DebugThickness=0.25)
    DriverAxis=(X=1,Y=0,Z=0)
    DriverItem=(Type=Bone,Name="foot_r")
    FalloffRegionScaleFactors=(PositiveWidth=3,NegativeWidth=0,PositiveHeight=1,NegativeHeight=1)
    FalloffSize=0.2
    FlipHeightScaling=False
    FlipWidthScaling=False
    OptionalParentItem=(Type=Bone,Name="None")
    RotationOffset=(X=90,Y=0,Z=-5)
    ExecutePin=@RigVMFunction_Sequence.C
   )
   > &RigUnit_ModifyTransforms(
    ItemToModify=((Item=(Type=Bone,Name="deform_foot_r"),Transform=(Rotation=(X=0.0561,Y=0.1525,Z=0.0858,W=0.983),Translation=(X=0.302,Y=-5.0869,Z=-7.6331),Scale3D=(X=1,Y=1,Z=0.25))))
    WeightMinimum=0
    Weight=@n11.OutputParam
   )
   > n12 = &RigUnit_SphericalPoseReader(
    ActiveRegionScaleFactors=(PositiveWidth=1,NegativeWidth=1,PositiveHeight=1,NegativeHeight=1)
    ActiveRegionSize=0.1
    Debug=(bDrawDebug=false,bDraw2D=False,bDrawLocalAxes=False,DebugScale=25,DebugSegments=20,DebugThickness=0.25)
    DriverAxis=(X=1,Y=0,Z=0)
    DriverItem=(Type=Bone,Name="foot_l")
    FalloffRegionScaleFactors=(PositiveWidth=3,NegativeWidth=0,PositiveHeight=1,NegativeHeight=1)
    FalloffSize=0.2
    FlipHeightScaling=False
    FlipWidthScaling=False
    OptionalParentItem=(Type=Bone,Name="None")
    RotationOffset=(X=90,Y=0,Z=-5)
   )
   > &RigUnit_ModifyTransforms(
    ItemToModify=((Item=(Type=Bone,Name="deform_foot_l"),Transform=(Rotation=(X=0.3292,Y=0.185,Z=-0.0012,W=0.926),Translation=(X=0.302,Y=-5.0872,Z=-7.6334),Scale3D=(X=1,Y=1,Z=0.25))))
    WeightMinimum=0
    Weight=@n12.OutputParam
   )
   RigVMFunction_ControlFlowBranch = &RigVMFunction_ControlFlowBranch(ExecuteContext=@RigVMFunction_Sequence.D, Condition=RigVMVariableNode() {variable=is DBNO}.Value)
   &RigUnit_ModifyTransforms(
    ItemToModify=((Item=(Type=Bone,Name="dyn_hoodie_bk_master"),Transform=(Rotation=(X=0,Y=0,Z=-0.1305,W=0.9914),Translation=(X=0,Y=0,Z=0),Scale3D=(X=1,Y=1,Z=1))))
    Weight=1
    WeightMinimum=0
    ExecutePin=@RigVMFunction_ControlFlowBranch.True
   )
   &RigUnit_ModifyTransforms(
    ExecutePin=()
    ItemToModify="((Item=(Type=Bone,Name=\"lowerarm_r\"),Transform=(Rotation=(X=0,Y=0,Z=0.8192,W=0.5736),Translation=(X=0,Y=0,Z=0),Scale3D=(X=1,Y=1,Z=1))),(Item=(Type=Bone,Name=\"lowerarm_l\"),Transform=(Rotation=(X=0,Y=0,Z=0.8192,W=0.5736),Translation=(X=0,Y=0,Z=0),Scale3D=(X= ...(truncated)"
    Weight=1
    WeightMinimum=0
   )
  ]}
]
hierarchy [
 root {loc=<0 0 0>, rot=<0 0 0 1>}
 pelvis {parent=root, loc=<0 -0.5301 98.0571>, rot=<0.0127 0.707 -0.0127 -0.707>}
 spine_01 {parent=pelvis, loc=<6.1208 -0.9328 0>, rot=<0 0 0.034 -0.9994>}
 spine_02 {parent=spine_01, loc=<6.5996 -0.2095 0>, rot=<0 0 0.0238 -0.9997>}
 spine_03 {parent=spine_02, loc=<6.4914 -0.1237 0>, rot=<0 0 0.0149 -0.9999>}
 spine_04 {parent=spine_03, loc=<6.9598 0.0794 0>, rot=<0 0 -0.0132 -0.9999>}
 spine_05 {parent=spine_04, loc=<6.9987 0 0>, rot=<0 0 -0.0015 -1>}
 clavicle_l {parent=spine_05, loc=<5.4831 -0.1053 -3.2946>, rot=<0.0465 -0.7289 -0.009 -0.6829>}
 upperarm_l {parent=clavicle_l, loc=<9.8658 -1.2136 -0.4105>, rot=<0.0164 -0.3245 -0.0269 -0.9454>}
 lowerarm_l {parent=upperarm_l, loc=<26.465 -0.8345 0.259>, rot=<0.0036 0.0161 -0.2298 -0.9731>}
 hand_l {parent=lowerarm_l, loc=<22.7355 -0.3203 -0.6105>, rot=<-0.6784 -0.0223 0.0735 -0.7307>}
 index_metacarpal_l {parent=hand_l, loc=<1.6535 0.0183 -2.96>, rot=<0.0158 -0.0124 -0.0173 -0.9996>}
 index_01_l {parent=index_metacarpal_l, loc=<8.3008 0.0357 -0.0524>, rot=<-0.0044 0.0485 0.0924 -0.9945>}
 index_02_l {parent=index_01_l, loc=<3.3587 -0.3004 0.0452>, rot=<-0.002 0.0027 0.3225 -0.9466>}
 index_03_l {parent=index_02_l, loc=<2.5578 -0.123 -0.0252>, rot=<-0.002 0.0067 0.2558 -0.9667>}
 middle_metacarpal_l {parent=hand_l, loc=<2.1856 -0.1725 -0.5984>, rot=<0.006 0.0168 -0.0329 -0.9993>}
 middle_01_l {parent=middle_metacarpal_l, loc=<7.3275 -0.2447 0.0129>, rot=<-0.0208 0.0042 0.0858 -0.9961>}
 middle_02_l {parent=middle_01_l, loc=<4.0123 0.1401 0.1175>, rot=<-0.0218 0.0258 0.3512 -0.9357>}
 middle_03_l {parent=middle_02_l, loc=<2.6077 0.0534 0.1373>, rot=<0.0082 0.0188 0.1708 -0.9851>}
 pinky_metacarpal_l {parent=hand_l, loc=<1.7616 -0.9753 3.3814>, rot=<-0.0397 0.0814 -0.069 -0.9935>}
 pinky_01_l {parent=pinky_metacarpal_l, loc=<7.1709 -0.4831 -0.1094>, rot=<-0.0654 -0.0035 0.2242 -0.9723>}
 pinky_02_l {parent=pinky_01_l, loc=<2.8274 0.0893 -0.1043>, rot=<0.0037 0.0011 0.2854 -0.9584>}
 pinky_03_l {parent=pinky_02_l, loc=<2.2046 0.0307 -0.0078>, rot=<-0.0002 0.0098 0.1445 -0.9895>}
 ring_metacarpal_l {parent=hand_l, loc=<2.2074 -0.3625 1.3638>, rot=<-0.0284 0.0722 -0.0458 -0.9959>}
 ring_01_l {parent=ring_metacarpal_l, loc=<7.0295 -0.6197 -0.1446>, rot=<-0.0489 -0.0252 0.1892 -0.9804>}
 ring_02_l {parent=ring_01_l, loc=<3.6174 0.4647 -0.1027>, rot=<0 0 0.2806 -0.9598>}
 ring_03_l {parent=ring_02_l, loc=<2.4152 0.114 0.0199>, rot=<0 0 0.1189 -0.9929>}
 thumb_01_l {parent=hand_l, loc=<1.3118 -0.3191 -2.459>, rot=<0.6546 -0.4042 0.0664 -0.6354>}
 thumb_02_l {parent=thumb_01_l, loc=<5.9336 -0.0448 0.0401>, rot=<-0.0037 0.0213 0.298 -0.9543>}
 thumb_03_l {parent=thumb_02_l, loc=<2.2223 0.3682 -0.0907>, rot=<0.0118 -0.013 0.0066 -0.9998>}
 deform_wrist_dn_base_l {parent=hand_l, loc=<0 -1.5326 0>, rot=<0 0 0 -1>}
 deform_wrist_dn_l {parent=deform_wrist_dn_base_l, loc=<0 0 0>, rot=<0 0 0 1>}
 deform_wrist_up_base_l {parent=hand_l, loc=<0 2.3009 0>, rot=<0 0 0 -1>}
 deform_wrist_up_l {parent=deform_wrist_up_base_l, loc=<0 0 0>, rot=<0 0 0 1>}
 lowerarm_twist_01_l {parent=lowerarm_l, loc=<18.316 0 0>, rot=<0 0 0 1>}
 lowerarm_twist_02_l {parent=lowerarm_l, loc=<10.6707 0 0>, rot=<0 0 0 1>}
 deform_elbow_in_base_l {parent=lowerarm_l, loc=<0.1834 3.6062 -0.2572>, rot=<0 0 0 -1>}
 deform_elbow_in_l {parent=deform_elbow_in_base_l, loc=<0 0 0>, rot=<0 0 0 1>}
 deform_elbow_out_base_l {parent=lowerarm_l, loc=<0 -4 0>, rot=<0 0 0 1>}
 deform_elbow_out_l {parent=deform_elbow_out_base_l, loc=<0 0 0>, rot=<0 0 0 1>}
 deform_sleeve_l {parent=lowerarm_l, loc=<0.0001 0 0>, rot=<0 0 0 1>}
 dyn_sleeveBk_l_master {parent=deform_sleeve_l, loc=<4.365 -6.7687 -3.2234>, rot=<-0.0798 0.0177 0.2157 -0.973>}
 dyn_sleeveBk_l_01 {parent=dyn_sleeveBk_l_master, loc=<0 0 0>, rot=<0 0 0 1>}
 dyn_sleeveFr_l_master {parent=deform_sleeve_l, loc=<5.3988 5.1382 0.0002>, rot=<0 0 -0.1305 -0.9914>}
 dyn_sleeveFr_l_01 {parent=dyn_sleeveFr_l_master, loc=<0 0 0>, rot=<0 0 0 1>}
 dyn_sleeveOut_l_master {parent=deform_sleeve_l, loc=<5.9485 0.3778 5.007>, rot=<0 0 0 -1>}
 dyn_sleeveOut_l_01 {parent=dyn_sleeveOut_l_master, loc=<0 0 0>, rot=<0 0 0 1>}
 dyn_sleeveIn_l_master {parent=deform_sleeve_l, loc=<5.063 0.3542 -7.4647>, rot=<-0.0123 -0.2516 -0.008 -0.9677>}
 dyn_sleeveIn_l_01 {parent=dyn_sleeveIn_l_master, loc=<0 0 0>, rot=<0 0 0 -1>}
 upperarm_twist_01_l {parent=upperarm_l, loc=<7.2135 -0.1134 0>, rot=<0 0 0 1>}
 deform_upperarm_bk_base_l {parent=upperarm_twist_01_l, loc=<-6.4053 -4.3746 0.314>, rot=<0 0 0 -1>}
 deform_upperarm_bk_l {parent=deform_upperarm_bk_base_l, loc=<0 0 0>, rot=<0 0 0 1>}
 deform_upperarm_fr_base_l {parent=upperarm_twist_01_l, loc=<-5.6312 5.092 0.3354>, rot=<0 0 0 -1>}
 deform_upperarm_fr_l {parent=deform_upperarm_fr_base_l, loc=<0 0 0>, rot=<0 0 0 1>}
 deform_upperarm_up_base_l {parent=upperarm_twist_01_l, loc=<-8.891 0.4825 3.9517>, rot=<0 0 0 -1>}
 deform_upperarm_up_l {parent=deform_upperarm_up_base_l, loc=<0 0 0>, rot=<0 0 0 1>}
 upperarm_twist_02_l {parent=upperarm_l, loc=<19.7998 -0.1134 0>, rot=<0 0 0 1>}
 clavicle_r {parent=spine_05, loc=<5.4831 -0.1053 3.2946>, rot=<0.7289 0.0465 -0.6829 0.009>}
 upperarm_r {parent=clavicle_r, loc=<-9.8658 1.2136 0.4104>, rot=<0.0164 -0.3245 -0.0269 -0.9454>}
 lowerarm_r {parent=upperarm_r, loc=<-26.4651 0.8345 -0.259>, rot=<0.0036 0.0161 -0.2298 -0.9731>}
 hand_r {parent=lowerarm_r, loc=<-22.7356 0.3203 0.6106>, rot=<-0.6784 -0.0223 0.0735 -0.7307>}
 index_metacarpal_r {parent=hand_r, loc=<-1.6534 -0.0181 2.9601>, rot=<0.0158 -0.0124 -0.0173 -0.9996>}
 index_01_r {parent=index_metacarpal_r, loc=<-8.3008 -0.0357 0.0524>, rot=<-0.0044 0.0485 0.0924 -0.9945>}
 index_02_r {parent=index_01_r, loc=<-3.3589 0.3003 -0.0453>, rot=<-0.002 0.0027 0.3225 -0.9466>}
 index_03_r {parent=index_02_r, loc=<-2.5578 0.123 0.0252>, rot=<-0.002 0.0067 0.2558 -0.9667>}
 middle_metacarpal_r {parent=hand_r, loc=<-2.1856 0.1726 0.5984>, rot=<0.006 0.0168 -0.0329 -0.9993>}
 middle_01_r {parent=middle_metacarpal_r, loc=<-7.3273 0.2447 -0.0129>, rot=<-0.0208 0.0042 0.0858 -0.9961>}
 middle_02_r {parent=middle_01_r, loc=<-4.0124 -0.14 -0.1175>, rot=<-0.0218 0.0258 0.3512 -0.9357>}
 middle_03_r {parent=middle_02_r, loc=<-2.6077 -0.0533 -0.1373>, rot=<0.0082 0.0188 0.1708 -0.9851>}
 pinky_metacarpal_r {parent=hand_r, loc=<-1.7617 0.9753 -3.3814>, rot=<-0.0397 0.0814 -0.069 -0.9935>}
 pinky_01_r {parent=pinky_metacarpal_r, loc=<-7.1707 0.4832 0.1094>, rot=<-0.0654 -0.0035 0.2242 -0.9723>}
 pinky_02_r {parent=pinky_01_r, loc=<-2.8274 -0.0892 0.1043>, rot=<0.0037 0.0011 0.2854 -0.9584>}
 pinky_03_r {parent=pinky_02_r, loc=<-2.2045 -0.031 0.0078>, rot=<-0.0002 0.0098 0.1445 -0.9895>}
 ring_metacarpal_r {parent=hand_r, loc=<-2.2073 0.3627 -1.3638>, rot=<-0.0284 0.0722 -0.0458 -0.9959>}
 ring_01_r {parent=ring_metacarpal_r, loc=<-7.0297 0.6197 0.1446>, rot=<-0.0489 -0.0252 0.1892 -0.9804>}
 ring_02_r {parent=ring_01_r, loc=<-3.6171 -0.4648 0.1027>, rot=<0 0 0.2806 -0.9598>}
 ring_03_r {parent=ring_02_r, loc=<-2.4153 -0.1141 -0.0199>, rot=<0 0 0.1189 -0.9929>}
 thumb_01_r {parent=hand_r, loc=<-1.3119 0.3192 2.459>, rot=<0.6546 -0.4042 0.0664 -0.6354>}
 thumb_02_r {parent=thumb_01_r, loc=<-5.9335 0.0447 -0.0401>, rot=<-0.0037 0.0213 0.298 -0.9543>}
 thumb_03_r {parent=thumb_02_r, loc=<-2.2223 -0.3682 0.0907>, rot=<0.0118 -0.013 0.0066 -0.9998>}
 deform_wrist_dn_base_r {parent=hand_r, loc=<0 1.5329 0>, rot=<0 0 0 -1>}
 deform_wrist_dn_r {parent=deform_wrist_dn_base_r, loc=<0 0 0>, rot=<0 0 0 1>}
 deform_wrist_up_base_r {parent=hand_r, loc=<-0.0001 -2.3006 -0.0001>, rot=<0 0 0 -1>}
 deform_wrist_up_r {parent=deform_wrist_up_base_r, loc=<0 0 0>, rot=<0 0 0 1>}
 lowerarm_twist_01_r {parent=lowerarm_r, loc=<-18.3161 0 0>, rot=<0 0 0 1>}
 lowerarm_twist_02_r {parent=lowerarm_r, loc=<-10.6707 0 0>, rot=<0 0 0 1>}
 deform_elbow_in_base_r {parent=lowerarm_r, loc=<-0.1834 -3.6062 0.2572>, rot=<0 0 0 -1>}
 deform_elbow_in_r {parent=deform_elbow_in_base_r, loc=<0 0 0>, rot=<0 0 0 1>}
 deform_elbow_out_base_r {parent=lowerarm_r, loc=<0 4 0>, rot=<0 0 0 1>}
 deform_elbow_out_r {parent=deform_elbow_out_base_r, loc=<0 0 0>, rot=<0 0 0 1>}
 deform_sleeve_r {parent=lowerarm_r, loc=<-0.0001 0 0.0001>, rot=<0 0 0 1>}
 dyn_sleeveBk_r_master {parent=deform_sleeve_r, loc=<-4.365 6.7687 3.2234>, rot=<-0.0798 0.0177 0.2157 -0.973>}
 dyn_sleeveBk_r_01 {parent=dyn_sleeveBk_r_master, loc=<0 0 0>, rot=<0 0 0 1>}
 dyn_sleeveFr_r_master {parent=deform_sleeve_r, loc=<-5.3988 -5.1382 -0.0002>, rot=<0 0 -0.1305 -0.9914>}
 dyn_sleeveFr_r_01 {parent=dyn_sleeveFr_r_master, loc=<0 0 0>, rot=<0 0 0 1>}
 dyn_sleeveOut_r_master {parent=deform_sleeve_r, loc=<-5.9485 -0.3778 -5.007>, rot=<0 0 0 1>}
 dyn_sleeveOut_r_01 {parent=dyn_sleeveOut_r_master, loc=<0 0 0>, rot=<0 0 0 1>}
 dyn_sleeveIn_r_master {parent=deform_sleeve_r, loc=<-5.063 -0.3542 7.4647>, rot=<-0.0123 -0.2516 -0.008 -0.9677>}
 dyn_sleeveIn_r_01 {parent=dyn_sleeveIn_r_master, loc=<0 0 0>, rot=<0 0 0 1>}
 upperarm_twist_01_r {parent=upperarm_r, loc=<-7.2135 0.1134 0>, rot=<0 0 0 1>}
 deform_upperarm_bk_base_r {parent=upperarm_twist_01_r, loc=<6.4052 4.3746 -0.3139>, rot=<0 0 0 -1>}
 deform_upperarm_bk_r {parent=deform_upperarm_bk_base_r, loc=<0 0 0>, rot=<0 0 0 1>}
 deform_upperarm_fr_base_r {parent=upperarm_twist_01_r, loc=<5.6311 -5.092 -0.3353>, rot=<0 0 0 -1>}
 deform_upperarm_fr_r {parent=deform_upperarm_fr_base_r, loc=<0 0 0>, rot=<0 0 0 1>}
 deform_upperarm_up_base_r {parent=upperarm_twist_01_r, loc=<8.8909 -0.4825 -3.9516>, rot=<0 0 0 -1>}
 deform_upperarm_up_r {parent=deform_upperarm_up_base_r, loc=<0 0 0>, rot=<0 0 0 1>}
 upperarm_twist_02_r {parent=upperarm_r, loc=<-19.7998 0.1134 0>, rot=<0 0 0 1>}
 neck_01 {parent=spine_05, loc=<11.3138 0.1529 0>, rot=<0 0 -0.2003 -0.9797>}
 neck_02 {parent=neck_01, loc=<4.5627 0 0>, rot=<0 0 0.0504 -0.9987>}
 head {parent=neck_02, loc=<5.3914 0 0>, rot=<0 0 0.115 -0.9934>}
 dyn_hood_back_master {parent=head, loc=<6.5797 -9.5902 0>, rot=<0.3115 -0.6348 0.6348 -0.3115>}
 dyn_hood_back {parent=dyn_hood_back_master, loc=<0 0 0>, rot=<0 0 0 1>}
 attach_cape {parent=spine_05, loc=<9.3216 -8.4501 0>, rot=<0.0616 -0.7044 -0.0616 -0.7044>}
 attach_backpack {parent=spine_05, loc=<0 -8.35 0>, rot=<-0.0093 -0.707 0.0093 -0.707>}
 deform_collar_front_master {parent=spine_05, loc=<4.9014 7.5985 0>, rot=<0 0 -0.2387 -0.9711>}
 deform_collar_front {parent=deform_collar_front_master, loc=<0 0 0>, rot=<0 0 0 1>}
 dyn_string_l_attach {parent=deform_collar_front, loc=<5.3895 0.305 -1.9755>, rot=<-0.1981 -0.9732 0.114 0.0256>}
 dyn_string_l_master {parent=dyn_string_l_attach, loc=<0.0237 0.2322 0.0396>, rot=<0 0 0 1>}
 dyn_string_l_01 {parent=dyn_string_l_master, loc=<0 0 0>, rot=<0 0 0 1>}
 dyn_string_l_02 {parent=dyn_string_l_01, loc=<3.1737 0 0.0019>, rot=<0 0.0043 0 -1>}
 dyn_string_l_03 {parent=dyn_string_l_02, loc=<2.3484 0 0>, rot=<0 0 0 1>}
 dyn_string_r_attach {parent=deform_collar_front, loc=<5.4091 0.2166 2.0787>, rot=<-0.1981 -0.9732 0.114 0.0256>}
 dyn_string_r_master {parent=dyn_string_r_attach, loc=<0.0762 0.285 -0.0565>, rot=<0 0 0 1>}
 dyn_string_r_01 {parent=dyn_string_r_master, loc=<0 0 0>, rot=<0 0 0 1>}
 dyn_string_r_02 {parent=dyn_string_r_01, loc=<3.1739 0.0061 0.0273>, rot=<0 0.0043 0 -1>}
 dyn_string_r_03 {parent=dyn_string_r_02, loc=<2.349 -0.0041 -0.017>, rot=<0 0 0 1>}
 deform_hood_l_master {parent=spine_05, loc=<14.1063 1.2187 -9.6152>, rot=<0.0284 -0.7065 -0.0284 -0.7065>}
 deform_hood_l {parent=deform_hood_l_master, loc=<0 0 0>, rot=<0 0 0 1>}
 deform_hood_r_master {parent=spine_05, loc=<14.1063 1.2187 9.6152>, rot=<0.7065 0.0284 0.7065 -0.0284>}
 deform_hood_r {parent=deform_hood_r_master, loc=<0 0 0>, rot=<0 0 0 1>}
 deform_hood_back_master {parent=spine_05, loc=<15.9783 -8.1738 0>, rot=<0.5196 -0.4795 0.4795 -0.5196>}
 deform_hood_back {parent=deform_hood_back_master, loc=<0 0 0>, rot=<0 0 0 1>}
 deform_collar_bk_master {parent=spine_05, loc=<12.564 -5.2601 0>, rot=<0.7024 0.0815 -0.0815 -0.7024>}
 deform_collar_bk {parent=deform_collar_bk_master, loc=<0 0 0>, rot=<0 0 0 1>}
 deform_collar_l_master {parent=spine_05, loc=<11.4059 1.7911 -6.8861>, rot=<0 0 -0.181 -0.9835>}
 deform_collar_l {parent=deform_collar_l_master, loc=<0 0 0>, rot=<0 0 0 1>}
 deform_collar_r_master {parent=spine_05, loc=<11.4059 1.7911 6.886>, rot=<0.9835 0.181 0 0>}
 deform_collar_r {parent=deform_collar_r_master, loc=<0 0 0>, rot=<0 0 0 1>}
 deform_pec_aim_base_l {parent=spine_04, loc=<3.2456 9.3309 -1.2237>, rot=<0.0603 -0.4033 0.3318 -0.8507>}
 deform_pec_aim_l {parent=deform_pec_aim_base_l, loc=<0 0 0>, rot=<0 0 0 1>}
 deform_pec_base_l {parent=deform_pec_aim_l, loc=<9.3665 0 0>, rot=<0 0 0 1>}
 deform_pec_l {parent=deform_pec_base_l, loc=<0 0 0>, rot=<0 0 0 1>}
 deform_pec_aim_base_r {parent=spine_04, loc=<3.2452 9.3309 1.2237>, rot=<-0.4033 -0.0603 0.8507 0.3318>}
 deform_pec_aim_r {parent=deform_pec_aim_base_r, loc=<0 0 0>, rot=<0 0 0 1>}
 deform_pec_base_r {parent=deform_pec_aim_r, loc=<-9.3667 -0.0003 -0.0004>, rot=<0 0 0 1>}
 deform_pec_r {parent=deform_pec_base_r, loc=<0 0 0>, rot=<0 0 0 1>}
 dyn_hoodie_l_master {parent=spine_03, loc=<0.6348 1.2268 -11.7848>, rot=<0.0673 -0.9977 0 0>}
 dyn_hoodie_l {parent=dyn_hoodie_l_master, loc=<0 0 0>, rot=<0 0 0 1>}
 dyn_hoodie_bk_master {parent=spine_03, loc=<2.2258 -6.8408 0>, rot=<0 0 0.9958 -0.0912>}
 dyn_hoodie_bk {parent=dyn_hoodie_bk_master, loc=<0 0 0>, rot=<0 0 0 1>}
 dyn_hoodie_fr_master {parent=spine_03, loc=<-1.8667 12.3099 0>, rot=<0.0799 -0.9968 0 0>}
 dyn_hoodie_fr {parent=dyn_hoodie_fr_master, loc=<0 0 0>, rot=<0 0 0 1>}
 dyn_hoodie_r_master {parent=spine_03, loc=<0.6349 1.2268 11.7848>, rot=<0.0673 -0.9977 0 0>}
 dyn_hoodie_r {parent=dyn_hoodie_r_master, loc=<0 0 0>, rot=<0 0 0 1>}
 thigh_l {parent=pelvis, loc=<-1.0855 -0.2109 -8.7053>, rot=<0.0009 0.0355 0.0155 -0.9993>}
 calf_l {parent=thigh_l, loc=<-42.5284 -0.0019 -0.7297>, rot=<0.0006 -0.0161 -0.0391 -0.9991>}
 calf_twist_01_l {parent=calf_l, loc=<-35.8483 -1.9197 -0.5794>, rot=<0.0237 0.0095 -0.0035 -0.9997>}
 calf_twist_02_l {parent=calf_l, loc=<-19.1546 -1.7943 -0.2638>, rot=<0.0237 0.0095 -0.0035 -0.9997>}
 foot_l {parent=calf_l, loc=<-46.7402 -2.5014 -0.592>, rot=<0.0427 -0.0238 0.0353 -0.9982>}
 ball_l {parent=foot_l, loc=<-3.5658 10.8066 -0.1544>, rot=<0.0001 -0.0001 -0.7063 -0.7079>}
 deform_foot_l_master {parent=foot_l, loc=<2.2282 -3.3321 -0.29>, rot=<-0.0357 -0.7045 -0.0284 -0.7082>}
 deform_foot_l {parent=deform_foot_l_master, loc=<0 0 0>, rot=<0 0 0 1>}
 deform_knee_in_base_l {parent=calf_l, loc=<0.8031 -5.0231 0.2456>, rot=<0 0 0 -1>}
 deform_knee_in_l {parent=deform_knee_in_base_l, loc=<0 0 0>, rot=<0 0 0 1>}
 deform_knee_out_base_l {parent=calf_l, loc=<0 4 0>, rot=<0 0 0 1>}
 deform_knee_out_l {parent=deform_knee_out_base_l, loc=<0 0 0>, rot=<0 0 0 1>}
 thigh_twist_01_l {parent=thigh_l, loc=<-27.952 0 0>, rot=<0.0013 0 0 -1>}
 deform_groin_base_l {parent=thigh_l, loc=<0.4801 5.6776 -0.2486>, rot=<0 0 0 -1>}
 deform_groin_l {parent=deform_groin_base_l, loc=<0 0 0>, rot=<0 0 0 1>}
 thigh_r {parent=pelvis, loc=<-1.0853 -0.2109 8.7053>, rot=<0.0355 -0.0009 0.9993 0.0155>}
 calf_r {parent=thigh_r, loc=<42.5286 0.0019 0.7297>, rot=<0.0006 -0.0161 -0.0391 -0.9991>}
 calf_twist_01_r {parent=calf_r, loc=<35.8485 1.9197 0.5794>, rot=<0.0237 0.0095 -0.0035 -0.9997>}
 calf_twist_02_r {parent=calf_r, loc=<19.1546 1.7943 0.2638>, rot=<0.0237 0.0095 -0.0035 -0.9997>}
 foot_r {parent=calf_r, loc=<46.7403 2.5014 0.592>, rot=<0.0427 -0.0238 0.0353 -0.9982>}
 ball_r {parent=foot_r, loc=<3.5658 -10.8066 0.1544>, rot=<0.0001 -0.0001 -0.7063 -0.7079>}
 deform_foot_r_master {parent=foot_r, loc=<-2.2284 3.3321 0.2898>, rot=<0.7082 0.0284 -0.7045 -0.0357>}
 deform_foot_r {parent=deform_foot_r_master, loc=<0 0 0>, rot=<0 0 0 1>}
 deform_knee_in_base_r {parent=calf_r, loc=<-0.8031 5.0231 -0.2457>, rot=<0 0 0 -1>}
 deform_knee_in_r {parent=deform_knee_in_base_r, loc=<0 0 0>, rot=<0 0 0 1>}
 deform_knee_out_base_r {parent=calf_r, loc=<0 -4 0>, rot=<0 0 0 1>}
 deform_knee_out_r {parent=deform_knee_out_base_r, loc=<0 0 0>, rot=<0 0 0 1>}
 thigh_twist_01_r {parent=thigh_r, loc=<27.952 0 0>, rot=<0.0013 0 0 -1>}
 deform_groin_base_r {parent=thigh_r, loc=<-0.4799 -5.6776 0.2485>, rot=<0 0 0 -1>}
 deform_groin_r {parent=deform_groin_base_r, loc=<0 0 0>, rot=<0 0 0 1>}
 dyn_device_master {parent=thigh_r, loc=<22.6206 -0.3871 7.5754>, rot=<0.0015 0.6816 0.0019 -0.7317>}
 dyn_device_01 {parent=dyn_device_master, loc=<0 0 0>, rot=<0 0 0 1>}
 deform_glute_base_l {parent=pelvis, loc=<-7.3466 -9.0213 -9.1245>, rot=<0.0172 0.031 -0.4477 -0.8935>}
 deform_glute_l {parent=deform_glute_base_l, loc=<0 0 0>, rot=<0 0 0 1>}
 deform_glute_base_r {parent=pelvis, loc=<-7.3464 -9.0213 9.1245>, rot=<0.031 -0.0172 0.8935 -0.4477>}
 deform_glute_r {parent=deform_glute_base_r, loc=<0 0 0>, rot=<0 0 0 1>}
 BendFwdCoeff {type=Curve, loc=<0 0 0>, rot=<0 0 0 1>}
 ElbowLBentInv {type=Curve, loc=<0 0 0>, rot=<0 0 0 1>}
 ElbowRBentInv {type=Curve, loc=<0 0 0>, rot=<0 0 0 1>}
]
variables [
 BendingFwdCoeff {type=double, default=0}
 LOD {type=int32, default=0}
 MathFloatRemap_1_1_1_4_Result {type=float, default=0}
 is DBNO {type=bool, default=True}
 isLoFi {type=bool, default=False}
]
