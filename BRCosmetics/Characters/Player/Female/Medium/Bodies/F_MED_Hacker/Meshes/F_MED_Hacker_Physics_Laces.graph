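# Physics asset for the collar laces on the F_MED_Hacker female player body, covering the left and right dyn_string bone chains.
# Each chain runs from a master through three small capsules with tight swing and twist limits, the left swinging less than the right.
# Both masters attach with free angular motion to a large front-collar capsule, which attaches likewise to spine_05.
# Large spine_04 and spine_05 capsules cover the torso, with spine_04 left unconstrained.
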PhysicsAsset /BRCosmetics/Characters/Player/Female/Medium/Bodies/F_MED_Hacker/Meshes/F_MED_Hacker_Physics_Laces
bodies {
 deform_collar_front_master [{
   shapeType=Capsule
   length=7.4632
   radius=10.2088
   shapeName=deform_collar_front_master_capsule
   center=<4.2782 -9.8326 -0.0026>
   rotation=<rot 0 -90.0001 -89.965>
  }]
 dyn_string_l_01 [{shapeType=Capsule, length=2, radius=1, shapeName=dyn_string_l_01_capsule, center=<2.1455 0.0043 0>, rotation=<rot 0 90.013 -89.7141>}]
 dyn_string_l_02 [{shapeType=Capsule, length=2, radius=1, shapeName=dyn_string_l_02_capsule, center=<1.1207 0.0088 0>, rotation=<rot 0 -89.9793 -89.783>}]
 dyn_string_l_03 [{shapeType=Capsule, length=0.8, radius=0.8, shapeName=dyn_string_l_03_capsule, center=<1.5976 0.009 0>, rotation=<rot 0 -89.8936 -89.9951>}]
 dyn_string_l_master [{shapeType=Capsule, length=0.505, radius=0.505, shapeName=dyn_string_l_master_capsule, rotation=<rot 0 -90 0>}]
 dyn_string_r_01 [{
   shapeType=Capsule
   length=1.6462
   radius=1
   shapeName=dyn_string_r_01_capsule
   center=<2.3126 0.0106 0.0234>
   rotation=<rot 0 90.0076 -89.7368>
  }]
 dyn_string_r_02 [{shapeType=Capsule, length=2, radius=1, shapeName=dyn_string_r_02_capsule, center=<1.5446 0.0063 -0.0015>, rotation=<rot 0 -89.9826 -89.769>}]
 dyn_string_r_03 [{
   shapeType=Capsule
   length=0.8
   radius=0.8
   shapeName=dyn_string_r_03_capsule
   center=<1.7632 0.0099 -0.0007>
   rotation=<rot 0 -90.0019 -89.5411>
  }]
 dyn_string_r_master [{shapeType=Capsule, length=0.505, radius=0.505, shapeName=dyn_string_r_master_capsule, rotation=<rot 0 -90 0>}]
 spine_04 [{
   shapeType=Capsule
   length=13.6621
   radius=10.856
   shapeName=spine_04_capsule
   center=<-0.121 1.7269 -0.0579>
   rotation=<rot -0.2353 -90 -0.4092>
  }]
 spine_05 [{
   shapeType=Capsule
   length=10.5361
   radius=9.3948
   shapeName=spine_05_capsule
   center=<-1.9554 0.5812 -0.0769>
   rotation=<rot 2.5581 -90.0002 177.1216>
  }]
}
constraints [
 {bone1Name=dyn_string_r_01, bone2Name=dyn_string_r_master, swing1LimitDegrees=25, swing2LimitDegrees=25, twistLimitDegrees=5}
 {bone1Name=dyn_string_r_02, bone2Name=dyn_string_r_01, swing1LimitDegrees=25, swing2LimitDegrees=25, twistLimitDegrees=5}
 {bone1Name=dyn_string_r_03, bone2Name=dyn_string_r_02, swing1LimitDegrees=25, swing2LimitDegrees=25, twistLimitDegrees=5}
 {bone1Name=dyn_string_l_01, bone2Name=dyn_string_l_master, swing1LimitDegrees=10, swing2LimitDegrees=10, twistLimitDegrees=5}
 {bone1Name=dyn_string_l_02, bone2Name=dyn_string_l_01, swing1LimitDegrees=10, swing2LimitDegrees=10, twistLimitDegrees=5}
 {bone1Name=dyn_string_l_03, bone2Name=dyn_string_l_02, swing1LimitDegrees=10, swing2LimitDegrees=10, twistLimitDegrees=5}
 {
  bone1Name=dyn_string_l_master
  bone2Name=deform_collar_front_master
  swing1LimitDegrees=45
  swing1Motion=Free
  swing2LimitDegrees=45
  swing2Motion=Free
  twistLimitDegrees=45
  twistMotion=Free
 }
 {
  bone1Name=dyn_string_r_master
  bone2Name=deform_collar_front_master
  swing1LimitDegrees=45
  swing1Motion=Free
  swing2LimitDegrees=45
  swing2Motion=Free
  twistLimitDegrees=45
  twistMotion=Free
 }
 {
  bone1Name=deform_collar_front_master
  bone2Name=spine_05
  swing1LimitDegrees=45
  swing1Motion=Free
  swing2LimitDegrees=45
  swing2Motion=Free
  twistLimitDegrees=45
  twistMotion=Free
 }
]
