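# Post-process style Animation Blueprint for the F_MED_Hacker cosmetic body, adding secondary motion atop the pose copied from its parent mesh.
# Control Rig, weighted bone constraints for device and string attachments, and two RigidBody simulations (body, laces) build a cached pose.
# Bool blends optionally add spring-driven AnimDynamics on sleeves and additive string-attach rotation offsets.
# Each update computes state-based gravity override, elbow curve values and LOD flags.
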AnimBlueprint /BRCosmetics/Characters/Player/Female/Medium/Bodies/F_MED_Hacker/Meshes/F_MED_Hacker_AnimBP
target_skeleton=&/F_MED_Hacker_Skeleton
graphs [
 AnimGraph {nodes [
   n1 = "Transform (Modify) Bone - Bone: dyn_string_l_attach"(ComponentPose=Local To Component(LocalPose=Use cached pose 'BasePose'()), Alpha=1, Translation="0,0,0", Rotation="9,0,-60", Scale="1,1,1") {
    rotationMode=BMM_Additive
    translationMode=BMM_Additive
    alphaBoolBlend {blendOption=Linear}
    boneToModify {boneName=dyn_string_l_attach}
    rotation {pitch=9, roll=-60}
   }
   n2 = "Transform (Modify) Bone - Bone: dyn_string_r_attach"(ComponentPose=@n1, Alpha=1, Translation="0,0,0", Rotation="0,0,-40", Scale="1,1,1") {rotationMode=BMM_Additive, alphaBoolBlend {blendOption=Linear}, boneToModify {boneName=dyn_string_r_attach}, rotation {roll=-40}}
   Output Pose(Result=Blend Poses by bool(BlendPose_0=Component To Local(ComponentPose=@n2), BlendPose_1=Use cached pose 'BasePose'()) {blendType=Linear}) {layerGroup=None}
   n3 = Anim Dynamics - Bone: dyn_sleeveBk_l_01(ComponentPose=Local To Component(LocalPose=Use cached pose 'PreAnimDyn'()), Alpha=1, ExternalForce="0,0,0") {
    angularDampingOverride=0.9999
    angularSpringConstant=200
    bAngularSpring=true
    bLinearSpring=true
    bOverrideAngularDamping=true
    bOverrideLinearDamping=true
    bUseGravityOverride=true
    linearDampingOverride=1
    linearSpringConstant=50
    alphaBoolBlend {blendOption=Linear}
    boundBone {boneName=dyn_sleeveBk_l_01}
    componentAppliedLinearAccClamp=<3000 3000 3000>
    componentLinearAccScale=<0.4 0.4 0.4>
    componentLinearVelScale=<0.4 0.4 0.4>
    gravityOverride {z=-980}
    physicsBodyDefinitions [{
      boundBone {boneName=dyn_sleeveBk_l_01}
      boxExtents=<10 10 10>
      constraintSetup {angularLimitsMax=<5 5 0>, angularLimitsMin=<-5 -5 -5>, angularTarget=<1 0 0>, linearAxesMax=<0 0.5 2>, linearAxesMin=<-1.5 0 0>}
      localJointOffset=<0.4 0 0>
     }]
   }
   n4 = Anim Dynamics - Bone: dyn_sleeveOut_l_01(ComponentPose=@n3, Alpha=1, ExternalForce="0,0,0") {
    angularDampingOverride=0.9999
    angularSpringConstant=200
    bAngularSpring=true
    bLinearSpring=true
    bOverrideAngularDamping=true
    bOverrideLinearDamping=true
    bUseGravityOverride=true
    linearDampingOverride=1
    linearSpringConstant=50
    alphaBoolBlend {blendOption=Linear}
    boundBone {boneName=dyn_sleeveOut_l_01}
    componentAppliedLinearAccClamp=<3000 3000 3000>
    componentLinearAccScale=<0.4 0.4 0.4>
    componentLinearVelScale=<0.4 0.4 0.4>
    gravityOverride {z=-980}
    physicsBodyDefinitions [{
      boundBone {boneName=dyn_sleeveOut_l_01}
      boxExtents=<10 10 10>
      constraintSetup {angularLimitsMax=<5 5 5>, angularLimitsMin=<-5 -5 -5>, angularTarget=<1 0 0>, linearAxesMax=<0 2.5 2>, linearAxesMin=<-1.5 0 0>}
      localJointOffset=<0.4 0 0>
     }]
   }
   n5 = Anim Dynamics - Bone: dyn_sleeveIn_l_01(ComponentPose=@n4, Alpha=1, ExternalForce="0,0,0") {
    angularDampingOverride=0.9999
    angularSpringConstant=200
    bAngularSpring=true
    bLinearSpring=true
    bOverrideAngularDamping=true
    bOverrideLinearDamping=true
    bUseGravityOverride=true
    linearDampingOverride=1
    linearSpringConstant=50
    alphaBoolBlend {blendOption=Linear}
    boundBone {boneName=dyn_sleeveIn_l_01}
    componentAppliedLinearAccClamp=<3000 3000 3000>
    componentLinearAccScale=<0.4 0.4 0.4>
    componentLinearVelScale=<0.4 0.4 0.4>
    gravityOverride {z=-980}
    physicsBodyDefinitions [{
      boundBone {boneName=dyn_sleeveIn_l_01}
      boxExtents=<10 10 10>
      constraintSetup {angularLimitsMax=<5 5 5>, angularLimitsMin=<-5 -5 -5>, angularTarget=<1 0 0>, linearAxesMax=<0 2 2>, linearAxesMin=<-2 0 0>}
      localJointOffset=<0.4 0 0>
     }]
   }
   n6 = Anim Dynamics - Bone: dyn_sleeveBk_r_01(ComponentPose=@n5, Alpha=1, ExternalForce="0,0,0") {
    angularDampingOverride=0.9999
    angularSpringConstant=200
    bAngularSpring=true
    bLinearSpring=true
    bOverrideAngularDamping=true
    bOverrideLinearDamping=true
    bUseGravityOverride=true
    linearDampingOverride=1
    linearSpringConstant=50
    alphaBoolBlend {blendOption=Linear}
    boundBone {boneName=dyn_sleeveBk_r_01}
    componentAppliedLinearAccClamp=<3000 3000 3000>
    componentLinearAccScale=<0.4 0.4 0.4>
    componentLinearVelScale=<0.4 0.4 0.4>
    gravityOverride {z=-980}
    physicsBodyDefinitions [{
      boundBone {boneName=dyn_sleeveBk_r_01}
      boxExtents=<10 10 10>
      constraintSetup {angularLimitsMax=<5 5 5>, angularLimitsMin=<-5 -5 0>, angularTarget=<1 0 0>, linearAxesMax=<1.5 0 0>, linearAxesMin=<0 -0.5 -2>}
      localJointOffset=<0.4 0 0>
     }]
   }
   n7 = Anim Dynamics - Bone: dyn_sleeveOut_r_01(ComponentPose=@n6, Alpha=1, ExternalForce="0,0,0") {
    angularDampingOverride=0.9999
    angularSpringConstant=200
    bAngularSpring=true
    bLinearSpring=true
    bOverrideAngularDamping=true
    bOverrideLinearDamping=true
    bUseGravityOverride=true
    linearDampingOverride=1
    linearSpringConstant=50
    alphaBoolBlend {blendOption=Linear}
    boundBone {boneName=dyn_sleeveOut_r_01}
    componentAppliedLinearAccClamp=<3000 3000 3000>
    componentLinearAccScale=<0.4 0.4 0.4>
    componentLinearVelScale=<0.4 0.4 0.4>
    gravityOverride {z=-980}
    physicsBodyDefinitions [{
      boundBone {boneName=dyn_sleeveOut_r_01}
      boxExtents=<10 10 10>
      constraintSetup {angularLimitsMax=<5 5 5>, angularLimitsMin=<-5 -5 -5>, angularTarget=<1 0 0>, linearAxesMax=<1.5 0 0>, linearAxesMin=<0 -1.5 -1.5>}
      localJointOffset=<0.4 0 0>
     }]
   }
   n8 = Anim Dynamics - Bone: dyn_sleeveIn_r_01(ComponentPose=@n7, Alpha=1, ExternalForce="0,0,0") {
    angularDampingOverride=0.9999
    angularSpringConstant=200
    bAngularSpring=true
    bLinearSpring=true
    bOverrideAngularDamping=true
    bOverrideLinearDamping=true
    bUseGravityOverride=true
    linearDampingOverride=1
    linearSpringConstant=50
    alphaBoolBlend {blendOption=Linear}
    boundBone {boneName=dyn_sleeveIn_r_01}
    componentAppliedLinearAccClamp=<3000 3000 3000>
    componentLinearAccScale=<0.4 0.4 0.4>
    componentLinearVelScale=<0.4 0.4 0.4>
    gravityOverride {z=-980}
    physicsBodyDefinitions [{
      boundBone {boneName=dyn_sleeveIn_r_01}
      boxExtents=<10 10 10>
      constraintSetup {angularLimitsMax=<5 5 5>, angularLimitsMin=<-5 -5 -5>, angularTarget=<1 0 0>, linearAxesMax=<2 0 0>, linearAxesMin=<0 -2 -2>}
      localJointOffset=<0.4 0 0>
     }]
   }
   Save cached pose 'BasePose'(Pose=Blend Poses by bool(BlendPose_0=Use cached pose 'PreAnimDyn'(), BlendPose_1=Component To Local(ComponentPose=@n8)) {blendType=Linear}) {cacheName=BasePose}
   n9 = Control Rig!(
    Source=Copy Pose From Mesh() {bUseAttachedParent=true}
    Alpha
    is DBNO:bool=Get bIsDBNO()
    isLoFi:bool=Get isLOD1_or_Higher() -> (isLOD1_or_Higher:bool)
    LOD:int=Get Current LOD() -> (Current LOD:int=0)
   ) -> (Pose) {
    alphaBoolBlend {blendOption=Linear}
    controlRigAssetReference {blueprintRigClass=&/F_MED_Hacker_CtrlRig.F_MED_Hacker_CtrlRig_C}
   }
   n10 = "Constraint - dyn_device_master (PP)"!(ComponentPose=Local To Component(LocalPose=@n9), Alpha=1, ConstraintWeights_0=0.2333, ConstraintWeights_1:float=0.7667) -> (Pose) {
    alphaBoolBlend {blendOption=Linear}
    boneToModify {boneName=dyn_device_master}
    constraintSetup [{targetBone {boneName=thigh_twist_01_r}}, {targetBone {boneName=thigh_r}}]
    constraintWeights [1,1]
   }
   n11 = "Constraint - dyn_string_r_attach (PPPPPPP)"!(
    ComponentPose=@n10
    Alpha=1
    ConstraintWeights_0=0.0021
    ConstraintWeights_1:float=0.0605
    ConstraintWeights_2:float=0.0016
    ConstraintWeights_3:float=0.0085
    ConstraintWeights_4:float=0.0061
    ConstraintWeights_5:float=0.0019
    ConstraintWeights_6:float=0.9193
   ) -> (Pose) {
    alphaBoolBlend {blendOption=Linear}
    boneToModify {boneName=dyn_string_r_attach}
    constraintSetup [
     {targetBone {boneName=neck_01}}
     {targetBone {boneName=spine_05}}
     {targetBone {boneName=clavicle_l}}
     {targetBone {boneName=clavicle_r}}
     {targetBone {boneName=spine_04}}
     {targetBone {boneName=deform_groin_r}}
     {targetBone {boneName=deform_collar_front}}
    ]
    constraintWeights [1,1,1,1,1,1,1]
   }
   n12 = "Constraint - dyn_string_l_attach (PPPPPPP)"!(
    ComponentPose=@n11
    Alpha=1
    ConstraintWeights_0=0.002
    ConstraintWeights_1:float=0.059
    ConstraintWeights_2:float=0.0081
    ConstraintWeights_3:float=0.0016
    ConstraintWeights_4:float=0.0059
    ConstraintWeights_5:float=0.0018
    ConstraintWeights_6:float=0.9214
   ) -> (Pose) {
    alphaBoolBlend {blendOption=Linear}
    boneToModify {boneName=dyn_string_l_attach}
    constraintSetup [
     {targetBone {boneName=neck_01}}
     {targetBone {boneName=spine_05}}
     {targetBone {boneName=clavicle_l}}
     {targetBone {boneName=clavicle_r}}
     {targetBone {boneName=spine_04}}
     {targetBone {boneName=deform_groin_l}}
     {targetBone {boneName=deform_collar_front}}
    ]
    constraintWeights [1,1,1,1,1,1,1]
   }
   n13 = RigidBody!(ComponentPose=@n12, Alpha, OverrideWorldGravity:Vector="0,0,0", ExternalForce) -> (Pose) {
    bClampLinearTranslationLimitToRefPose=true
    bOverrideWorldGravity=true
    overridePhysicsAsset=&/F_MED_Hacker_Physics
    simulationSpace=BaseBoneSpace
    alphaBoolBlend {blendOption=Linear}
    baseBoneRef {boneName=pelvis}
    componentAppliedLinearAccClamp=<1000 1000 1000>
    componentLinearAccScale=<0.4 0.4 0.4>
    componentLinearVelScale=<0.4 0.4 0.4>
    simSpaceSettings {
     maxAngularAcceleration=500
     maxAngularVelocity=3
     maxLinearAcceleration=400
     maxLinearVelocity=250
     velocityScaleZ=0
     worldAlpha=1
     externalLinearDragV=<0.2 0.2 0.2>
    }
   }
   n14 = RigidBody(ComponentPose=@n13, OverrideWorldGravity="0,0,0") {
    bClampLinearTranslationLimitToRefPose=true
    bOverrideWorldGravity=true
    overridePhysicsAsset=&/F_MED_Hacker_Physics_Laces
    simulationSpace=BaseBoneSpace
    alphaBoolBlend {blendOption=Linear}
    baseBoneRef {boneName=spine_05}
    componentAppliedLinearAccClamp=<1500 1500 1500>
    componentLinearAccScale=<0.5 0.5 0.5>
    componentLinearVelScale=<0.5 0.5 0.5>
    simSpaceSettings {
     maxAngularAcceleration=500
     maxAngularVelocity=3
     maxLinearAcceleration=400
     maxLinearVelocity=250
     velocityScaleZ=0
     worldAlpha=1
     externalLinearDragV=<0.2 0.2 0.2>
    }
   }
   Save cached pose 'PreAnimDyn'(Pose=Component To Local(ComponentPose=@n14)) {cacheName=PreAnimDyn}
  ]}
 EventGraph {nodes [
   Event BlueprintUpdateAnimation()
   > n1 = Sequence!(execute) -> (then_0, then_1, then_2:Exec)
   > n2 = Gravity Override State Function DEPRECATED(
    CharacterPartAnimInstance=Self-Reference()
    GameTimeInSeconds=GetGameTimeInSeconds()
    GravityOverrideParams=Get Gravity Override Params()
    RBAN Params=Get RBAN Params()
    DebugDraw=Get doDebug() -> (doDebug:bool)
   )
   > Set GravityOverride(GravityOverride=@n2)
   n3 = Get CurrentLOD()
   n4 = Set isLOD1_or_Higher(execute=@n1.then_2, isLOD1_or_Higher:bool="integer >= integer"(A=@n3, B=1)) -> (Output_Get:bool)
   > Set IsFrontendLoFi(IsFrontendLoFi:bool=AND Boolean(A=Get bIsInFrontEnd(), B=@n4)) -> (Output_Get:bool)
   > Set Current LOD(Current LOD:int=@n3) -> (Output_Get:int=0)
   Set ElbowLBentInv(execute=@n1.then_1, ElbowLBentInv:float=GetCurveValue(CurveName=ElbowLBentInv)) -> (Output_Get:float)
   > Set ElbowRBentInv(ElbowRBentInv:float=GetCurveValue(CurveName=ElbowRBentInv)) -> (Output_Get:float)
  ]}
]
variables [
 Current LOD {type=integer}
 doDebug {type=boolean}
 ElbowLBentInv {type=number}
 ElbowRBentInv {type=number}
 Gravity Override Params {type=GravityOverrideParamsStruct}
 GravityOverride {type=Vector}
 IsFrontendLoFi {type=boolean}
 isLOD1_or_Higher {type=boolean}
 RBAN Params {
  type=RBANParamsStruct
  default {
   dBNO {simSpaceSettings {
     maxAngularAcceleration=1000
     maxAngularVelocity=3
     maxLinearAcceleration=300
     maxLinearVelocity=700
     velocityScaleZ=0
     externalLinearDragV=<0 0.5 0>
    }}
   dBNO_Carried {simSpaceSettings {
     maxAngularAcceleration=1000
     maxAngularVelocity=3
     maxLinearAcceleration=300
     maxLinearVelocity=700
     velocityScaleZ=0
     externalLinearDragV=<0 0.5 0>
    }}
   emoteOrMelee {simSpaceSettings {
     maxAngularAcceleration=1000
     maxAngularVelocity=3
     maxLinearAcceleration=300
     maxLinearVelocity=700
     velocityScaleZ=0
     externalLinearDragV=<0 0.5 0>
    }}
   front_End {simSpaceSettings {
     maxAngularAcceleration=1000
     maxAngularVelocity=3
     maxLinearAcceleration=300
     maxLinearVelocity=700
     velocityScaleZ=0
     externalLinearDragV=<0 0.5 0>
    }}
   onGround_Crouch_Moving {simSpaceSettings {
     maxAngularAcceleration=1000
     maxAngularVelocity=3
     maxLinearAcceleration=300
     maxLinearVelocity=700
     velocityScaleZ=0
     externalLinearDragV=<0 0.5 0>
    }}
   onGround_Falling {simSpaceSettings {
     maxAngularAcceleration=1000
     maxAngularVelocity=3
     maxLinearAcceleration=300
     maxLinearVelocity=700
     velocityScaleZ=0
     externalLinearDragV=<0 0.5 0>
    }}
   onGround_Moving {simSpaceSettings {
     maxAngularAcceleration=1000
     maxAngularVelocity=3
     maxLinearAcceleration=300
     maxLinearVelocity=700
     velocityScaleZ=0
     externalLinearDragV=<0 0.5 0>
    }}
   onGround_Standing {simSpaceSettings {
     maxAngularAcceleration=1000
     maxAngularVelocity=3
     maxLinearAcceleration=300
     maxLinearVelocity=700
     velocityScaleZ=0
     externalLinearDragV=<0 0.5 0>
    }}
   rail_Grinding {simSpaceSettings {
     maxAngularAcceleration=1000
     maxAngularVelocity=3
     maxLinearAcceleration=300
     maxLinearVelocity=700
     velocityScaleZ=0
     worldAlpha=0.1
     externalLinearDragV=<0 0.5 0>
    }}
   riding_Moving {simSpaceSettings {
     maxAngularAcceleration=1000
     maxAngularVelocity=3
     maxLinearAcceleration=300
     maxLinearVelocity=700
     velocityScaleZ=0
     externalLinearDragV=<0 0.5 0>
    }}
   riding_Moving_Targeting {simSpaceSettings {
     maxAngularAcceleration=1000
     maxAngularVelocity=3
     maxLinearAcceleration=300
     maxLinearVelocity=700
     velocityScaleZ=0
     externalLinearDragV=<0 0.5 0>
    }}
   riding_Standing {simSpaceSettings {
     maxAngularAcceleration=1000
     maxAngularVelocity=3
     maxLinearAcceleration=300
     maxLinearVelocity=700
     velocityScaleZ=0
     externalLinearDragV=<0 0.5 0>
    }}
   skydive_DivingDown {simSpaceSettings {
     maxAngularAcceleration=1000
     maxAngularVelocity=3
     maxLinearAcceleration=300
     maxLinearVelocity=700
     velocityScaleZ=0
     externalLinearDragV=<0 0.5 0>
    }}
   skydive_FloatingDown {
    componentLinearAccScale=<0 0.5 0>
    componentLinearVelScale=<0 0.5 0>
    simSpaceSettings {
     maxAngularAcceleration=1000
     maxAngularVelocity=3
     maxLinearAcceleration=300
     maxLinearVelocity=700
     velocityScaleZ=0
     externalLinearDragV=<0 0.5 0>
    }
   }
   skydive_GoingUp {simSpaceSettings {
     maxAngularAcceleration=1000
     maxAngularVelocity=3
     maxLinearAcceleration=300
     maxLinearVelocity=700
     velocityScaleZ=0
     externalLinearDragV=<0 0.5 0>
    }}
   skydive_Parachuting {
    componentLinearAccScale=<0 0.1 0.3>
    componentLinearVelScale=<0 0.1 0.3>
    simSpaceSettings {
     maxAngularAcceleration=1000
     maxAngularVelocity=3
     maxLinearAcceleration=300
     maxLinearVelocity=700
     velocityScaleZ=0
     externalLinearDragV=<0 0.5 0>
    }
   }
   skydive_Surfing {simSpaceSettings {
     maxAngularAcceleration=1000
     maxAngularVelocity=3
     maxLinearAcceleration=300
     maxLinearVelocity=700
     velocityScaleZ=0
     externalLinearDragV=<0 0.5 0>
    }}
   swimming {simSpaceSettings {
     maxAngularAcceleration=1000
     maxAngularVelocity=3
     maxLinearAcceleration=300
     maxLinearVelocity=700
     velocityScaleZ=0
     externalLinearDragV=<0 0.5 0>
    }}
   tactical_Sprint {simSpaceSettings {
     maxAngularAcceleration=1000
     maxAngularVelocity=3
     maxLinearAcceleration=300
     maxLinearVelocity=700
     velocityScaleZ=0
     externalLinearDragV=<0 0.5 0>
    }}
   ziplining {simSpaceSettings {
     maxAngularAcceleration=1000
     maxAngularVelocity=3
     maxLinearAcceleration=300
     maxLinearVelocity=700
     velocityScaleZ=0
     externalLinearDragV=<0 0.5 0>
    }}
  }
 }
]
