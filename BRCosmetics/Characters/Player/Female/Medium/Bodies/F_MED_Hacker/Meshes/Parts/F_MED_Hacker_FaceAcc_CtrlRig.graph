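# Control Rig for the F_MED_Hacker face accessory part that procedurally corrects hood and hair bones as the body moves.
# A sequence runs parent constraints blending hood and hair masters between head, spine_05 and hood bones.
# Weighted ModifyTransforms correctives are driven by spherical pose readers on head and clavicles, plus remapped head rotations and pelvis height.
# An extra upper back hood corrective, driven by head-to-neck rotation, runs only when isPlayingEmote is true.
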ControlRigBlueprint /BRCosmetics/Characters/Player/Female/Medium/Bodies/F_MED_Hacker/Meshes/Parts/F_MED_Hacker_FaceAcc_CtrlRig
graphs [
 Multiply_6_ContainedGraph {nodes [
   n1 = RigVMFunctionEntryNode()
   RigVMFunctionReturnNode(Result=&RigVMFunction_MathDoubleMul(B=@n1.C, A=&RigVMFunction_MathDoubleMul(A=@n1.A, B=@n1.B).Result).Result)
  ]}
 RigVMFunctionLibrary {links [], nodes []}
 RigVMFunction_Sequence_1_ContainedGraph {nodes [
   n1 = RigVMFunctionEntryNode()
   RigVMFunction_Sequence = &RigVMFunction_Sequence(ExecuteContext=@n1.ExecuteContext)
   n2 = &RigVMFunction_Sequence(ExecuteContext=@RigVMFunction_Sequence.B)
   n3 = &RigVMFunction_Sequence(ExecuteContext=@n2.B)
   n4 = &RigVMFunction_Sequence(ExecuteContext=@n3.B)
   n5 = &RigVMFunction_Sequence(ExecuteContext=@n4.B)
   n6 = &RigVMFunction_Sequence(ExecuteContext=@n5.B)
   RigVMFunctionReturnNode(A=@RigVMFunction_Sequence.A, B=@n2.A, C=@n3.A, D=@n4.A, E=@n5.A, F=@n6.A, G=@n6.B)
  ]}
 RigVMModel {nodes [
   RigUnit_BeginExecution = &RigUnit_BeginExecution()
   n1 = &RigVMFunction_Sequence(ExecuteContext=@RigUnit_BeginExecution.ExecutePin)
   RigVMFunction_Sequence = RigVMAggregateNode(ExecuteContext=@n1.B)
   n2 = &RigUnit_ParentConstraint(
    AdvancedSettings=(InterpolationType=Average,RotationOrderForFilter=XZY)
    Child=(Type=Bone,Name="dyn_hair_l_master")
    Parents=((Item=(Type=Bone,Name="deform_hood_l"),Weight=0.5),(Item=(Type=Bone,Name="spine_05"),Weight=0.5))
    Weight=1
    ExecutePin=@RigVMFunction_Sequence.E
   )
   n3 = &RigUnit_SphericalPoseReader(
    ActiveRegionScaleFactors=(PositiveWidth=1,NegativeWidth=1,PositiveHeight=1,NegativeHeight=1)
    ActiveRegionSize=0.1
    Debug=(bDrawDebug=True,bDraw2D=False,bDrawLocalAxes=False,DebugScale=25,DebugSegments=20,DebugThickness=0.25)
    DriverAxis=(X=1,Y=0,Z=0)
    DriverItem=(Type=Bone,Name="clavicle_l")
    FalloffRegionScaleFactors=(PositiveWidth=1,NegativeWidth=1,PositiveHeight=1,NegativeHeight=1)
    FalloffSize=0.2
    FlipHeightScaling=False
    FlipWidthScaling=False
    OptionalParentItem=(Type=Bone,Name="None")
    RotationOffset=(X=45,Y=0,Z=-90)
    ExecutePin=@RigVMFunction_Sequence.D
   )
   &RigUnit_ModifyTransforms(
    ItemToModify=((Item=(Type=Bone,Name="dyn_hair_l_master"),Transform=(Rotation=(X=-0.1294,Y=-0.017,Z=0.1294,W=0.983),Translation=(X=0,Y=0,Z=0),Scale3D=(X=1,Y=1,Z=1))))
    WeightMinimum=0
    ExecutePin=@n2.ExecutePin
    Weight=@n3.OutputParam
   )
   > n4 = &RigUnit_ParentConstraint(
    AdvancedSettings=(InterpolationType=Average,RotationOrderForFilter=XZY)
    Child=(Type=Bone,Name="dyn_hair_r_master")
    Parents=((Item=(Type=Bone,Name="deform_hood_r"),Weight=0.5),(Item=(Type=Bone,Name="spine_05"),Weight=0.5))
    Weight=1
   )
   n5 = &RigUnit_SphericalPoseReader(
    ActiveRegionScaleFactors=(PositiveWidth=1,NegativeWidth=1,PositiveHeight=1,NegativeHeight=1)
    ActiveRegionSize=0.1
    Debug=(bDrawDebug=True,bDraw2D=False,bDrawLocalAxes=False,DebugScale=25,DebugSegments=20,DebugThickness=0.25)
    DriverAxis=(X=1,Y=0,Z=0)
    DriverItem=(Type=Bone,Name="clavicle_r")
    FalloffRegionScaleFactors=(PositiveWidth=1,NegativeWidth=1,PositiveHeight=1,NegativeHeight=1)
    FalloffSize=0.2
    FlipHeightScaling=False
    FlipWidthScaling=False
    OptionalParentItem=(Type=Bone,Name="None")
    RotationOffset=(X=45,Y=0,Z=-90)
    ExecutePin=@n15.ExecutePin
   )
   &RigUnit_ModifyTransforms(
    ItemToModify=((Item=(Type=Bone,Name="dyn_hair_r_master"),Transform=(Rotation=(X=-0.13,Y=-0.0114,Z=0.0864,W=0.9877),Translation=(X=0,Y=0,Z=0),Scale3D=(X=1,Y=1,Z=1))))
    WeightMinimum=0
    ExecutePin=@n4.ExecutePin
    Weight=@n5.OutputParam
   )
   RigVMFunction_MathQuaternionToEuler = &RigVMFunction_MathQuaternionToEuler(
    RotationOrder=ZYX
    Value=&RigUnit_GetRelativeTransformForItem(Child=(Type=Bone,Name="head"), Parent=(Type=Bone,Name="spine_05"), bChildInitial=False, bParentInitial=False).RelativeTransform.Rotation
   )
   n6 = &RigVMFunction_MathDoubleRemap(SourceMaximum=-90, SourceMinimum=-45, TargetMaximum=1, TargetMinimum=0, Value=@RigVMFunction_MathQuaternionToEuler.Result.X)
   > &RigUnit_ModifyTransforms(
    ItemToModify="((Item=(Type=Bone,Name=\"dyn_hair_r_master\"),Transform=(Rotation=(X=0,Y=0.139,Z=0,W=0.9903),Translation=(X=-0.3133,Y=0,Z=-0.9965),Scale3D=(X=1,Y=1,Z=1))),(Item=(Type=Bone,Name=\"dyn_hair_l_master\"),Transform=(Rotation=(X=-0.2526,Y=-0.0271,Z=0.1032,W=0.9617), ...(truncated)"
    WeightMinimum=0
    Weight=@n6.Result
   )
   n7 = &RigVMFunction_MathDoubleRemap(SourceMaximum=90, SourceMinimum=45, TargetMaximum=1, TargetMinimum=0, Value=@RigVMFunction_MathQuaternionToEuler.Result.X)
   > &RigUnit_ModifyTransforms(
    ItemToModify="((Item=(Type=Bone,Name=\"dyn_hair_l_master\"),Transform=(Rotation=(X=-0.071,Y=0.1449,Z=-0.1466,W=0.9759),Translation=(X=-0.3133,Y=0,Z=-0.9965),Scale3D=(X=1,Y=1,Z=1))),(Item=(Type=Bone,Name=\"dyn_hair_r_master\"),Transform=(Rotation=(X=-0.3655,Y=-0.0392,Z=0.099 ...(truncated)"
    WeightMinimum=0
    Weight=@n7.Result
   )
   > n8 = &RigUnit_SphericalPoseReader(
    ActiveRegionScaleFactors=(PositiveWidth=1.5,NegativeWidth=1,PositiveHeight=1,NegativeHeight=1)
    ActiveRegionSize=0.1
    Debug=(bDrawDebug=True,bDraw2D=False,bDrawLocalAxes=False,DebugScale=25,DebugSegments=20,DebugThickness=0.25)
    DriverAxis=(X=1,Y=0,Z=0)
    DriverItem=(Type=Bone,Name="head")
    FalloffRegionScaleFactors=(PositiveWidth=3,NegativeWidth=1,PositiveHeight=1,NegativeHeight=1)
    FalloffSize=0.2
    FlipHeightScaling=False
    FlipWidthScaling=False
    OptionalParentItem=(Type=Bone,Name="spine_05")
    RotationOffset=(X=-17.7532,Y=-28.3946,Z=-10)
   )
   > &RigUnit_ModifyTransforms(
    ItemToModify=((Item=(Type=Bone,Name="dyn_hair_r_master"),Transform=(Rotation=(X=0.1138,Y=0.1471,Z=-0.155,W=0.9703),Translation=(X=-0.0716,Y=-0.302,Z=0.0487),Scale3D=(X=1,Y=1,Z=1))))
    WeightMinimum=0
    Weight=@n8.OutputParam
   )
   RigVMFunction_ControlFlowBranch = &RigVMFunction_ControlFlowBranch(ExecuteContext=@RigVMFunction_Sequence.G, Condition=RigVMVariableNode() {variable=isPlayingEmote}.Value)
   n9 = &RigUnit_GetRelativeTransformForItem(Child=(Type=Bone,Name="head"), Parent=(Type=Bone,Name="neck_01"), bChildInitial=False, bParentInitial=False)
   &RigUnit_ModifyTransforms(
    ItemToModify=((Item=(Type=Bone,Name="deform_hoodUpper_back"),Transform=(Rotation=(X=0,Y=-0.0963,Z=0,W=0.9953),Translation=(X=-2.7112,Y=0,Z=-1.4429),Scale3D=(X=1,Y=1,Z=1))))
    WeightMinimum=0
    Weight=&RigVMFunction_MathDoubleRemap(
     SourceMaximum=-55
     SourceMinimum=-40
     TargetMaximum=1
     TargetMinimum=0
     bClamp=False
     Value=&RigVMFunction_MathQuaternionToEuler(RotationOrder=ZYX, Value=@n9.RelativeTransform.Rotation).Result.Z
    ).Result
    ExecutePin=@RigVMFunction_ControlFlowBranch.True
   )
   &RigUnit_ParentConstraint(
    AdvancedSettings=(InterpolationType=Average,RotationOrderForFilter=XZY)
    Child=(Type=Bone,Name="deform_hood_back_master")
    Parents=((Item=(Type=Bone,Name="head"),Weight=0.5),(Item=(Type=Bone,Name="spine_05"),Weight=0.5))
    Weight=1
    ExecutePin=@RigVMFunction_Sequence.B
   )
   > &RigUnit_ParentConstraint(
    AdvancedSettings=(InterpolationType=Average,RotationOrderForFilter=XZY)
    Child=(Type=Bone,Name="deform_hood_backUpper_master")
    Parents=((Item=(Type=Bone,Name="head"),Weight=0.7),(Item=(Type=Bone,Name="spine_05"),Weight=0.3))
    Weight=1
   )
   > Get Delta Ref = RigVMFunctionReferenceNode(Item=(Type=Bone,Name="head"))
   Remap = &RigVMFunction_MathDoubleRemap(SourceMaximum=-45, SourceMinimum=0, TargetMaximum=1, TargetMinimum=0, Value=@"Get Delta Ref".Result.Rotation.Yaw)
   > &RigUnit_ModifyTransforms(
    ItemToModify="((Item=(Type=Bone,Name=\"deform_hood_back\"),Transform=(Rotation=(X=0,Y=0,Z=0,W=1),Translation=(X=-2,Y=0,Z=2),Scale3D=(X=1,Y=1,Z=1))),(Item=(Type=Bone,Name=\"deform_hoodUpper_back\"),Transform=(Rotation=(X=0,Y=0,Z=0,W=1),Translation=(X=6,Y=0,Z=5),Scale3D=(X=1, ...(truncated)"
    WeightMinimum=0
    Weight=@Remap.Result
   )
   n10 = &RigVMFunction_MathDoubleRemap(SourceMaximum=-85, SourceMinimum=-45, TargetMaximum=1, TargetMinimum=0, Value=@RigVMFunction_MathQuaternionToEuler.Result.Z)
   n11 = &RigUnit_GetRelativeTransformForItem(Child=(Type=Bone,Name="pelvis"), Parent=(Type=Bone,Name="root"), bChildInitial=False, bParentInitial=False)
   > &RigUnit_ModifyTransforms(
    ItemToModify="((Item=(Type=Bone,Name=\"dyn_hood_back_master\"),Transform=(Rotation=(X=0,Y=-0.1491,Z=0,W=0.9888),Translation=(X=-2.174,Y=0,Z=-2.9172),Scale3D=(X=1,Y=1,Z=1))),(Item=(Type=Bone,Name=\"deform_hood_back\"),Transform=(Rotation=(X=0.0576,Y=0,Z=0,W=0.9983),Translati ...(truncated)"
    WeightMinimum=0
    Weight=&RigVMFunction_MathDoubleMul(
     A=@n10.Result
     B=&RigVMFunction_MathDoubleRemap(SourceMaximum=45, SourceMinimum=55, TargetMaximum=1, TargetMinimum=0, Value=@n11.RelativeTransform.Translation.Z).Result
    ).Result
   )
   n12 = &RigVMFunction_MathDoubleRemap(
    SourceMaximum=60
    SourceMinimum=50
    TargetMaximum=1
    TargetMinimum=0
    Value=&RigUnit_GetRelativeTransformForItem(Child=(Type=Bone,Name="pelvis"), Parent=(Type=Bone,Name="root"), bChildInitial=False, bParentInitial=False).RelativeTransform.Translation.Z
   )
   n13 = &RigVMFunction_MathDoubleRemap(
    SourceMaximum=77
    SourceMinimum=85
    TargetMaximum=1
    TargetMinimum=0
    Value=&RigUnit_GetRelativeTransformForItem(Child=(Type=Bone,Name="pelvis"), Parent=(Type=Bone,Name="root"), bChildInitial=False, bParentInitial=False).RelativeTransform.Translation.Z
   )
   > &RigUnit_ModifyTransforms(
    ItemToModify="((Item=(Type=Bone,Name=\"dyn_hood_back_master\"),Transform=(Rotation=(X=0,Y=-0.053,Z=0,W=0.9986),Translation=(X=-2.1976,Y=0,Z=-2.8407),Scale3D=(X=1,Y=1,Z=1))),(Item=(Type=Bone,Name=\"deform_hood_back\"),Transform=(Rotation=(X=0.0576,Y=0,Z=0,W=0.9983),Translati ...(truncated)"
    WeightMinimum=0
    Weight=&RigVMFunction_MathDoubleMul(B=&RigVMFunction_MathDoubleMul(B=@n12.Result, A=@n13.Result).Result, A=@n10.Result).Result
   )
   &RigUnit_ParentConstraint(
    AdvancedSettings=(InterpolationType=Average,RotationOrderForFilter=XZY)
    Child=(Type=Bone,Name="deform_hood_l")
    Parents=((Item=(Type=Bone,Name="spine_05"),Weight=0.5),(Item=(Type=Bone,Name="head"),Weight=0.5))
    Weight=1
    ExecutePin=@RigVMFunction_Sequence.C
   )
   > &RigUnit_ParentConstraint(
    AdvancedSettings=(InterpolationType=Average,RotationOrderForFilter=XZY)
    Child=(Type=Bone,Name="deform_hood_r")
    Parents=((Item=(Type=Bone,Name="spine_05"),Weight=0.5),(Item=(Type=Bone,Name="head"),Weight=0.5))
    Weight=1
   )
   Multiply = &RigVMFunction_MathDoubleMul(
    A=&RigVMFunction_MathDoubleRemap(SourceMaximum=16, SourceMinimum=0, TargetMaximum=1, TargetMinimum=0, Value=@"Get Delta Ref".Result.Rotation.Pitch).Result
    B=&RigVMFunction_MathDoubleRemap(SourceMaximum=-20, SourceMinimum=0, TargetMaximum=1, TargetMinimum=0, Value=@"Get Delta Ref".Result.Rotation.Yaw).Result
   )
   > &RigUnit_ModifyTransforms(
    ItemToModify=((Item=(Type=Bone,Name="deform_hood_r"),Transform=(Rotation=(X=0,Y=0.2467,Z=0,W=0.9691),Translation=(X=4.9348,Y=0,Z=0),Scale3D=(X=1,Y=1,Z=1))))
    WeightMinimum=0
    Weight=@Multiply.Result
   )
   n14 = &RigVMFunction_MathQuaternionToEuler(
    RotationOrder=ZYX
    Value=&RigUnit_GetRelativeTransformForItem(Child=(Type=Bone,Name="head"), Parent=(Type=Bone,Name="spine_05"), bChildInitial=False, bParentInitial=False).RelativeTransform.Rotation
   )
   Multiply_5 = RigVMAggregateNode(
    C=&RigVMFunction_MathDoubleRemap(SourceMaximum=-25, SourceMinimum=-35, TargetMaximum=1, TargetMinimum=0, Value=@n14.Result.Z).Result
    A=&RigVMFunction_MathDoubleRemap(SourceMaximum=-35, SourceMinimum=-25, TargetMaximum=1, TargetMinimum=0, Value=@n14.Result.X).Result
    B=&RigVMFunction_MathDoubleRemap(SourceMaximum=-15, SourceMinimum=0, TargetMaximum=1, TargetMinimum=0, Value=@n14.Result.Z).Result
   )
   &RigUnit_ModifyTransforms(
    ItemToModify=((Item=(Type=Bone,Name="dyn_hair_r_master"),Transform=(Rotation=(X=0,Y=0.0872,Z=0,W=0.9962),Translation=(X=0,Y=0,Z=0),Scale3D=(X=1,Y=1,Z=1))))
    WeightMinimum=0
    Weight=@Multiply_5.Result
    ExecutePin=@RigVMFunction_Sequence.F
   )
   &RigUnit_ModifyTransforms(
    ExecutePin=()
    ItemToModify="((Item=(Type=Bone,Name=\"clavicle_l\"),Transform=(Rotation=(X=0,Y=-0.342,Z=0,W=0.9397),Translation=(X=0,Y=0,Z=0),Scale3D=(X=1,Y=1,Z=1))),(Item=(Type=Bone,Name=\"clavicle_r\"),Transform=(Rotation=(X=0,Y=-0.342,Z=0,W=0.9397),Translation=(X=0,Y=0,Z=0),Scale3D=(X= ...(truncated)"
    Weight=1
    WeightMinimum=0
   )
   &RigUnit_ModifyTransforms(
    ExecutePin=()
    ItemToModify="((Item=(Type=Bone,Name=\"neck_01\"),Transform=(Rotation=(X=0.2588,Y=0,Z=0,W=0.9659),Translation=(X=0,Y=0,Z=0),Scale3D=(X=1,Y=1,Z=1))),(Item=(Type=Bone,Name=\"neck_02\"),Transform=(Rotation=(X=0.2588,Y=0,Z=0,W=0.9659),Translation=(X=0,Y=0,Z=0),Scale3D=(X=1,Y=1, ...(truncated)"
    Weight=1
    WeightMinimum=0
   )
   &RigUnit_ModifyTransforms(
    ItemToModify=((Item=(Type=Bone,Name="deform_hood_r"),Transform=(Rotation=(X=-0.0872,Y=-0.0073,Z=-0.0834,W=0.9927),Translation=(X=-7.301,Y=0,Z=0),Scale3D=(X=1,Y=1,Z=1))))
    WeightMinimum=0
    ExecutePin=@n16.ExecutePin
    Weight=&RigVMFunction_MathDoubleMul(
     B=&RigVMFunction_MathDoubleRemap(SourceMaximum=-85, SourceMinimum=0, TargetMaximum=1, TargetMinimum=0, Value=@RigVMFunction_MathQuaternionToEuler.Result.Z).Result
     A=@n3.OutputParam
    ).Result
   )
   n15 = &RigUnit_ModifyTransforms(
    ItemToModify=((Item=(Type=Bone,Name="deform_hood_l"),Transform=(Rotation=(X=0.0038,Y=0.0435,Z=-0.0871,W=0.9952),Translation=(X=0,Y=-3,Z=2),Scale3D=(X=1,Y=1,Z=1))))
    WeightMinimum=0
    ExecutePin=@n3.ExecutePin
    Weight=@n3.OutputParam
   )
   &RigUnit_ModifyTransforms(
    ItemToModify="((Item=(Type=Bone,Name=\"deform_hood_r\"),Transform=(Rotation=(X=0,Y=0,Z=-0.0875,W=0.9962),Translation=(X=6.9328,Y=-1.2275,Z=0),Scale3D=(X=1,Y=1,Z=1))),(Item=(Type=Bone,Name=\"deform_hood_l\"),Transform=(Rotation=(X=0,Y=0,Z=0,W=1),Translation=(X=3.3846,Y=0,Z=0 ...(truncated)"
    WeightMinimum=0
    ExecutePin=@n17.ExecutePin
    Weight=@n6.Result
   )
   > n16 = &RigUnit_ModifyTransforms(
    ItemToModify="((Item=(Type=Bone,Name=\"deform_hood_l\"),Transform=(Rotation=(X=0,Y=0,Z=-0.0875,W=0.9962),Translation=(X=6.9328,Y=-1.2275,Z=0),Scale3D=(X=1,Y=1,Z=1))),(Item=(Type=Bone,Name=\"deform_hood_r\"),Transform=(Rotation=(X=0,Y=0,Z=0,W=1),Translation=(X=3.3846,Y=0,Z=0 ...(truncated)"
    WeightMinimum=0
    Weight=@n7.Result
   )
   n17 = &RigUnit_ModifyTransforms(
    ItemToModify=((Item=(Type=Bone,Name="deform_hood_r"),Transform=(Rotation=(X=-0.13,Y=-0.0114,Z=0.0864,W=0.9877),Translation=(X=0,Y=3,Z=2),Scale3D=(X=1,Y=1,Z=1))))
    WeightMinimum=0
    ExecutePin=@n5.ExecutePin
    Weight=@n5.OutputParam
   )
  ]}
]
hierarchy [
 root {loc=<0 0 0>, rot=<0 0 0 1>}
 pelvis {parent=root, loc=<0 -0.5301 98.0571>, rot=<0.0127 0.707 -0.0127 -0.707>}
 spine_01 {parent=pelvis, loc=<6.1208 -0.9328 0>, rot=<0 0 0.034 -0.9994>}
 spine_02 {parent=spine_01, loc=<6.5996 -0.2095 0>, rot=<0 0 0.0238 -0.9997>}
 spine_03 {parent=spine_02, loc=<6.4914 -0.1237 0>, rot=<0 0 0.0149 -0.9999>}
 spine_04 {parent=spine_03, loc=<6.9598 0.0794 0>, rot=<0 0 -0.0132 -0.9999>}
 spine_05 {parent=spine_04, loc=<6.9987 0 0>, rot=<0 0 -0.0015 -1>}
 clavicle_l {parent=spine_05, loc=<5.4831 -0.1053 -3.2946>, rot=<0.0465 -0.7289 -0.009 -0.6829>}
 upperarm_l {parent=clavicle_l, loc=<9.8658 -1.2136 -0.4105>, rot=<0.0164 -0.3245 -0.0269 -0.9454>}
 lowerarm_l {parent=upperarm_l, loc=<26.465 -0.8345 0.259>, rot=<0.0036 0.0161 -0.2298 -0.9731>}
 hand_l {parent=lowerarm_l, loc=<22.7355 -0.3203 -0.6105>, rot=<-0.6784 -0.0223 0.0735 -0.7307>}
 index_metacarpal_l {parent=hand_l, loc=<1.6535 0.0183 -2.96>, rot=<0.0158 -0.0124 -0.0173 -0.9996>}
 index_01_l {parent=index_metacarpal_l, loc=<8.3008 0.0357 -0.0524>, rot=<-0.0044 0.0485 0.0924 -0.9945>}
 index_02_l {parent=index_01_l, loc=<3.3587 -0.3004 0.0452>, rot=<-0.002 0.0027 0.3225 -0.9466>}
 index_03_l {parent=index_02_l, loc=<2.5578 -0.123 -0.0252>, rot=<-0.002 0.0067 0.2558 -0.9667>}
 middle_metacarpal_l {parent=hand_l, loc=<2.1856 -0.1725 -0.5984>, rot=<0.006 0.0168 -0.0329 -0.9993>}
 middle_01_l {parent=middle_metacarpal_l, loc=<7.3275 -0.2447 0.0129>, rot=<-0.0208 0.0042 0.0858 -0.9961>}
 middle_02_l {parent=middle_01_l, loc=<4.0123 0.1401 0.1175>, rot=<-0.0218 0.0258 0.3512 -0.9357>}
 middle_03_l {parent=middle_02_l, loc=<2.6077 0.0534 0.1373>, rot=<0.0082 0.0188 0.1708 -0.9851>}
 pinky_metacarpal_l {parent=hand_l, loc=<1.7616 -0.9753 3.3814>, rot=<-0.0397 0.0814 -0.069 -0.9935>}
 pinky_01_l {parent=pinky_metacarpal_l, loc=<7.1709 -0.4831 -0.1094>, rot=<-0.0654 -0.0035 0.2242 -0.9723>}
 pinky_02_l {parent=pinky_01_l, loc=<2.8274 0.0893 -0.1043>, rot=<0.0037 0.0011 0.2854 -0.9584>}
 pinky_03_l {parent=pinky_02_l, loc=<2.2046 0.0307 -0.0078>, rot=<-0.0002 0.0098 0.1445 -0.9895>}
 ring_metacarpal_l {parent=hand_l, loc=<2.2074 -0.3625 1.3638>, rot=<-0.0284 0.0722 -0.0458 -0.9959>}
 ring_01_l {parent=ring_metacarpal_l, loc=<7.0295 -0.6197 -0.1446>, rot=<-0.0489 -0.0252 0.1892 -0.9804>}
 ring_02_l {parent=ring_01_l, loc=<3.6174 0.4647 -0.1027>, rot=<0 0 0.2806 -0.9598>}
 ring_03_l {parent=ring_02_l, loc=<2.4152 0.114 0.0199>, rot=<0 0 0.1189 -0.9929>}
 thumb_01_l {parent=hand_l, loc=<1.3118 -0.3191 -2.459>, rot=<0.6546 -0.4042 0.0664 -0.6354>}
 thumb_02_l {parent=thumb_01_l, loc=<5.9336 -0.0448 0.0401>, rot=<-0.0037 0.0213 0.298 -0.9543>}
 thumb_03_l {parent=thumb_02_l, loc=<2.2223 0.3682 -0.0907>, rot=<0.0118 -0.013 0.0066 -0.9998>}
 deform_wrist_dn_base_l {parent=hand_l, loc=<0 -1.5326 0>, rot=<0 0 0 -1>}
 deform_wrist_dn_l {parent=deform_wrist_dn_base_l, loc=<0 0 0>, rot=<0 0 0 1>}
 deform_wrist_up_base_l {parent=hand_l, loc=<0 2.3009 0>, rot=<0 0 0 -1>}
 deform_wrist_up_l {parent=deform_wrist_up_base_l, loc=<0 0 0>, rot=<0 0 0 1>}
 weapon_l {parent=hand_l, loc=<-0.1974 -3.4111 0.3811>, rot=<0 0 0 1>}
 lowerarm_twist_01_l {parent=lowerarm_l, loc=<18.316 0 0>, rot=<0 0 0 1>}
 lowerarm_twist_02_l {parent=lowerarm_l, loc=<10.6707 0 0>, rot=<0 0 0 1>}
 deform_elbow_in_base_l {parent=lowerarm_l, loc=<0.1834 3.6062 -0.2572>, rot=<0 0 0 -1>}
 deform_elbow_in_l {parent=deform_elbow_in_base_l, loc=<0 0 0>, rot=<0 0 0 1>}
 deform_elbow_out_base_l {parent=lowerarm_l, loc=<0 -4 0>, rot=<0 0 0 1>}
 deform_elbow_out_l {parent=deform_elbow_out_base_l, loc=<0 0 0>, rot=<0 0 0 1>}
 deform_sleeve_l {parent=lowerarm_l, loc=<0.0001 0 0>, rot=<0 0 0 1>}
 dyn_sleeveBk_l_master {parent=deform_sleeve_l, loc=<4.365 -6.7687 -3.2234>, rot=<-0.0798 0.0177 0.2157 -0.973>}
 dyn_sleeveBk_l_01 {parent=dyn_sleeveBk_l_master, loc=<0 0 0>, rot=<0 0 0 1>}
 dyn_sleeveFr_l_master {parent=deform_sleeve_l, loc=<5.3988 5.1382 0.0002>, rot=<0 0 -0.1305 -0.9914>}
 dyn_sleeveFr_l_01 {parent=dyn_sleeveFr_l_master, loc=<0 0 0>, rot=<0 0 0 1>}
 dyn_sleeveOut_l_master {parent=deform_sleeve_l, loc=<5.9485 0.3778 5.007>, rot=<0 0 0 -1>}
 dyn_sleeveOut_l_01 {parent=dyn_sleeveOut_l_master, loc=<0 0 0>, rot=<0 0 0 1>}
 dyn_sleeveIn_l_master {parent=deform_sleeve_l, loc=<5.063 0.3542 -7.4647>, rot=<-0.0123 -0.2516 -0.008 -0.9677>}
 dyn_sleeveIn_l_01 {parent=dyn_sleeveIn_l_master, loc=<0 0 0>, rot=<0 0 0 -1>}
 upperarm_twist_01_l {parent=upperarm_l, loc=<7.2135 -0.1134 0>, rot=<0 0 0 1>}
 deform_upperarm_bk_base_l {parent=upperarm_twist_01_l, loc=<-6.4053 -4.3746 0.314>, rot=<0 0 0 -1>}
 deform_upperarm_bk_l {parent=deform_upperarm_bk_base_l, loc=<0 0 0>, rot=<0 0 0 1>}
 deform_upperarm_fr_base_l {parent=upperarm_twist_01_l, loc=<-5.6312 5.092 0.3354>, rot=<0 0 0 -1>}
 deform_upperarm_fr_l {parent=deform_upperarm_fr_base_l, loc=<0 0 0>, rot=<0 0 0 1>}
 deform_upperarm_up_base_l {parent=upperarm_twist_01_l, loc=<-8.891 0.4825 3.9517>, rot=<0 0 0 -1>}
 deform_upperarm_up_l {parent=deform_upperarm_up_base_l, loc=<0 0 0>, rot=<0 0 0 1>}
 upperarm_twist_02_l {parent=upperarm_l, loc=<19.7998 -0.1134 0>, rot=<0 0 0 1>}
 clavicle_r {parent=spine_05, loc=<5.4831 -0.1053 3.2946>, rot=<0.7289 0.0465 -0.6829 0.009>}
 upperarm_r {parent=clavicle_r, loc=<-9.8658 1.2136 0.4104>, rot=<0.0164 -0.3245 -0.0269 -0.9454>}
 lowerarm_r {parent=upperarm_r, loc=<-26.4651 0.8345 -0.259>, rot=<0.0036 0.0161 -0.2298 -0.9731>}
 hand_r {parent=lowerarm_r, loc=<-22.7356 0.3203 0.6106>, rot=<-0.6784 -0.0223 0.0735 -0.7307>}
 index_metacarpal_r {parent=hand_r, loc=<-1.6534 -0.0181 2.9601>, rot=<0.0158 -0.0124 -0.0173 -0.9996>}
 index_01_r {parent=index_metacarpal_r, loc=<-8.3008 -0.0357 0.0524>, rot=<-0.0044 0.0485 0.0924 -0.9945>}
 index_02_r {parent=index_01_r, loc=<-3.3589 0.3003 -0.0453>, rot=<-0.002 0.0027 0.3225 -0.9466>}
 index_03_r {parent=index_02_r, loc=<-2.5578 0.123 0.0252>, rot=<-0.002 0.0067 0.2558 -0.9667>}
 middle_metacarpal_r {parent=hand_r, loc=<-2.1856 0.1726 0.5984>, rot=<0.006 0.0168 -0.0329 -0.9993>}
 middle_01_r {parent=middle_metacarpal_r, loc=<-7.3273 0.2447 -0.0129>, rot=<-0.0208 0.0042 0.0858 -0.9961>}
 middle_02_r {parent=middle_01_r, loc=<-4.0124 -0.14 -0.1175>, rot=<-0.0218 0.0258 0.3512 -0.9357>}
 middle_03_r {parent=middle_02_r, loc=<-2.6077 -0.0533 -0.1373>, rot=<0.0082 0.0188 0.1708 -0.9851>}
 pinky_metacarpal_r {parent=hand_r, loc=<-1.7617 0.9753 -3.3814>, rot=<-0.0397 0.0814 -0.069 -0.9935>}
 pinky_01_r {parent=pinky_metacarpal_r, loc=<-7.1707 0.4832 0.1094>, rot=<-0.0654 -0.0035 0.2242 -0.9723>}
 pinky_02_r {parent=pinky_01_r, loc=<-2.8274 -0.0892 0.1043>, rot=<0.0037 0.0011 0.2854 -0.9584>}
 pinky_03_r {parent=pinky_02_r, loc=<-2.2045 -0.031 0.0078>, rot=<-0.0002 0.0098 0.1445 -0.9895>}
 ring_metacarpal_r {parent=hand_r, loc=<-2.2073 0.3627 -1.3638>, rot=<-0.0284 0.0722 -0.0458 -0.9959>}
 ring_01_r {parent=ring_metacarpal_r, loc=<-7.0297 0.6197 0.1446>, rot=<-0.0489 -0.0252 0.1892 -0.9804>}
 ring_02_r {parent=ring_01_r, loc=<-3.6171 -0.4648 0.1027>, rot=<0 0 0.2806 -0.9598>}
 ring_03_r {parent=ring_02_r, loc=<-2.4153 -0.1141 -0.0199>, rot=<0 0 0.1189 -0.9929>}
 thumb_01_r {parent=hand_r, loc=<-1.3119 0.3192 2.459>, rot=<0.6546 -0.4042 0.0664 -0.6354>}
 thumb_02_r {parent=thumb_01_r, loc=<-5.9335 0.0447 -0.0401>, rot=<-0.0037 0.0213 0.298 -0.9543>}
 thumb_03_r {parent=thumb_02_r, loc=<-2.2223 -0.3682 0.0907>, rot=<0.0118 -0.013 0.0066 -0.9998>}
 deform_wrist_dn_base_r {parent=hand_r, loc=<0 1.5329 0>, rot=<0 0 0 -1>}
 deform_wrist_dn_r {parent=deform_wrist_dn_base_r, loc=<0 0 0>, rot=<0 0 0 1>}
 deform_wrist_up_base_r {parent=hand_r, loc=<-0.0001 -2.3006 -0.0001>, rot=<0 0 0 -1>}
 deform_wrist_up_r {parent=deform_wrist_up_base_r, loc=<0 0 0>, rot=<0 0 0 1>}
 weapon_r {parent=hand_r, loc=<0.1975 3.4112 -0.3811>, rot=<0 0 0 -1>}
 lowerarm_twist_01_r {parent=lowerarm_r, loc=<-18.3161 0 0>, rot=<0 0 0 1>}
 lowerarm_twist_02_r {parent=lowerarm_r, loc=<-10.6707 0 0>, rot=<0 0 0 1>}
 deform_elbow_in_base_r {parent=lowerarm_r, loc=<-0.1834 -3.6062 0.2572>, rot=<0 0 0 -1>}
 deform_elbow_in_r {parent=deform_elbow_in_base_r, loc=<0 0 0>, rot=<0 0 0 1>}
 deform_elbow_out_base_r {parent=lowerarm_r, loc=<0 4 0>, rot=<0 0 0 1>}
 deform_elbow_out_r {parent=deform_elbow_out_base_r, loc=<0 0 0>, rot=<0 0 0 1>}
 deform_sleeve_r {parent=lowerarm_r, loc=<-0.0001 0 0.0001>, rot=<0 0 0 1>}
 dyn_sleeveBk_r_master {parent=deform_sleeve_r, loc=<-4.365 6.7687 3.2234>, rot=<-0.0798 0.0177 0.2157 -0.973>}
 dyn_sleeveBk_r_01 {parent=dyn_sleeveBk_r_master, loc=<0 0 0>, rot=<0 0 0 1>}
 dyn_sleeveFr_r_master {parent=deform_sleeve_r, loc=<-5.3988 -5.1382 -0.0002>, rot=<0 0 -0.1305 -0.9914>}
 dyn_sleeveFr_r_01 {parent=dyn_sleeveFr_r_master, loc=<0 0 0>, rot=<0 0 0 1>}
 dyn_sleeveOut_r_master {parent=deform_sleeve_r, loc=<-5.9485 -0.3778 -5.007>, rot=<0 0 0 1>}
 dyn_sleeveOut_r_01 {parent=dyn_sleeveOut_r_master, loc=<0 0 0>, rot=<0 0 0 1>}
 dyn_sleeveIn_r_master {parent=deform_sleeve_r, loc=<-5.063 -0.3542 7.4647>, rot=<-0.0123 -0.2516 -0.008 -0.9677>}
 dyn_sleeveIn_r_01 {parent=dyn_sleeveIn_r_master, loc=<0 0 0>, rot=<0 0 0 1>}
 upperarm_twist_01_r {parent=upperarm_r, loc=<-7.2135 0.1134 0>, rot=<0 0 0 1>}
 deform_upperarm_bk_base_r {parent=upperarm_twist_01_r, loc=<6.4052 4.3746 -0.3139>, rot=<0 0 0 -1>}
 deform_upperarm_bk_r {parent=deform_upperarm_bk_base_r, loc=<0 0 0>, rot=<0 0 0 1>}
 deform_upperarm_fr_base_r {parent=upperarm_twist_01_r, loc=<5.6311 -5.092 -0.3353>, rot=<0 0 0 -1>}
 deform_upperarm_fr_r {parent=deform_upperarm_fr_base_r, loc=<0 0 0>, rot=<0 0 0 1>}
 deform_upperarm_up_base_r {parent=upperarm_twist_01_r, loc=<8.8909 -0.4825 -3.9516>, rot=<0 0 0 -1>}
 deform_upperarm_up_r {parent=deform_upperarm_up_base_r, loc=<0 0 0>, rot=<0 0 0 1>}
 upperarm_twist_02_r {parent=upperarm_r, loc=<-19.7998 0.1134 0>, rot=<0 0 0 1>}
 neck_01 {parent=spine_05, loc=<11.3138 0.1529 0>, rot=<0 0 -0.2003 -0.9797>}
 neck_02 {parent=neck_01, loc=<4.5627 0 0>, rot=<0 0 0.0504 -0.9987>}
 head {parent=neck_02, loc=<5.3914 0 0>, rot=<0 0 0.115 -0.9934>}
 dyn_hood_back_master {parent=head, loc=<6.5797 -9.5902 0>, rot=<0.3115 -0.6348 0.6348 -0.3115>}
 dyn_hood_back {parent=dyn_hood_back_master, loc=<0 0 0>, rot=<0 0 0 1>}
 deform_hood_l_master {parent=spine_05, loc=<14.1063 1.2187 -9.6152>, rot=<0.0284 -0.7065 -0.0284 -0.7065>}
 deform_hood_l {parent=deform_hood_l_master, loc=<0 0 0>, rot=<0 0 0 1>}
 deform_hood_r_master {parent=spine_05, loc=<14.1063 1.2187 9.6152>, rot=<0.7065 0.0284 0.7065 -0.0284>}
 deform_hood_r {parent=deform_hood_r_master, loc=<0 0 0>, rot=<0 0 0 1>}
 deform_hood_back_master {parent=spine_05, loc=<15.9783 -8.1738 0>, rot=<0.5196 -0.4795 0.4795 -0.5196>}
 deform_hood_back {parent=deform_hood_back_master, loc=<0 0 0>, rot=<0 0 0 1>}
 attach_cape {parent=spine_05, loc=<9.3216 -8.4501 0>, rot=<0.0616 -0.7044 -0.0616 -0.7044>}
 attach_backpack {parent=spine_05, loc=<0 -8.35 0>, rot=<-0.0093 -0.707 0.0093 -0.707>}
 deform_collar_front_master {parent=spine_05, loc=<4.9014 7.5985 0>, rot=<0 0 -0.2387 -0.9711>}
 deform_collar_front {parent=deform_collar_front_master, loc=<0 0 0>, rot=<0 0 0 1>}
 dyn_string_l_attach {parent=deform_collar_front, loc=<5.3895 0.305 -1.9755>, rot=<-0.1981 -0.9732 0.114 0.0256>}
 dyn_string_l_master {parent=dyn_string_l_attach, loc=<0.0237 0.2322 0.0396>, rot=<0 0 0 1>}
 dyn_string_l_01 {parent=dyn_string_l_master, loc=<0 0 0>, rot=<0 0 0 1>}
 dyn_string_l_02 {parent=dyn_string_l_01, loc=<3.1737 0 0.0019>, rot=<0 0.0043 0 -1>}
 dyn_string_l_03 {parent=dyn_string_l_02, loc=<2.3484 0 0>, rot=<0 0 0 1>}
 dyn_string_r_attach {parent=deform_collar_front, loc=<5.4091 0.2166 2.0787>, rot=<-0.1981 -0.9732 0.114 0.0256>}
 dyn_string_r_master {parent=dyn_string_r_attach, loc=<0.0762 0.285 -0.0565>, rot=<0 0 0 1>}
 dyn_string_r_01 {parent=dyn_string_r_master, loc=<0 0 0>, rot=<0 0 0 1>}
 dyn_string_r_02 {parent=dyn_string_r_01, loc=<3.1739 0.0061 0.0273>, rot=<0 0.0043 0 -1>}
 dyn_string_r_03 {parent=dyn_string_r_02, loc=<2.349 -0.0041 -0.017>, rot=<0 0 0 1>}
 deform_collar_bk_master {parent=spine_05, loc=<12.564 -5.2601 0>, rot=<0.7024 0.0815 -0.0815 -0.7024>}
 deform_collar_bk {parent=deform_collar_bk_master, loc=<0 0 0>, rot=<0 0 0 1>}
 deform_collar_l_master {parent=spine_05, loc=<11.4059 1.7911 -6.8861>, rot=<0 0 -0.181 -0.9835>}
 deform_collar_l {parent=deform_collar_l_master, loc=<0 0 0>, rot=<0 0 0 1>}
 deform_collar_r_master {parent=spine_05, loc=<11.4059 1.7911 6.886>, rot=<0.9835 0.181 0 0>}
 deform_collar_r {parent=deform_collar_r_master, loc=<0 0 0>, rot=<0 0 0 1>}
 dyn_hair_r_master {parent=spine_05, loc=<14.1797 4.7639 8.2801>, rot=<-0.3205 -0.9083 0.2406 -0.1199>}
 dyn_hair_r_01 {parent=dyn_hair_r_master, loc=<0 0 0>, rot=<0 0 0 1>}
 dyn_hair_r_02 {parent=dyn_hair_r_01, loc=<3.8251 -0.0963 0.0201>, rot=<-0.0003 0.0081 0.039 -0.9992>}
 dyn_hair_r_03 {parent=dyn_hair_r_02, loc=<3.1917 0 0>, rot=<0 0 0 1>}
 dyn_hair_l_master {parent=spine_05, loc=<14.5603 5.056 -9.0633>, rot=<0.9134 -0.2919 -0.117 -0.2584>}
 dyn_hair_l_01 {parent=dyn_hair_l_master, loc=<0 0 0>, rot=<0 0 0 1>}
 dyn_hair_l_02 {parent=dyn_hair_l_01, loc=<-3.825 0.0963 -0.0201>, rot=<-0.0003 0.0081 0.039 -0.9992>}
 dyn_hair_l_03 {parent=dyn_hair_l_02, loc=<-3.1919 0.0002 0>, rot=<0 0 0 1>}
 deform_hood_backUpper_master {parent=spine_05, loc=<20.7456 -6.4848 0>, rot=<0.5196 -0.4795 0.4795 -0.5196>}
 deform_hoodUpper_back {parent=deform_hood_backUpper_master, loc=<0 0 0>, rot=<0 0 0 1>}
 deform_pec_aim_base_l {parent=spine_04, loc=<3.2456 9.3309 -1.2237>, rot=<0.0603 -0.4033 0.3318 -0.8507>}
 deform_pec_aim_l {parent=deform_pec_aim_base_l, loc=<0 0 0>, rot=<0 0 0 1>}
 deform_pec_base_l {parent=deform_pec_aim_l, loc=<9.3665 0 0>, rot=<0 0 0 1>}
 deform_pec_l {parent=deform_pec_base_l, loc=<0 0 0>, rot=<0 0 0 1>}
 deform_pec_aim_base_r {parent=spine_04, loc=<3.2452 9.3309 1.2237>, rot=<-0.4033 -0.0603 0.8507 0.3318>}
 deform_pec_aim_r {parent=deform_pec_aim_base_r, loc=<0 0 0>, rot=<0 0 0 1>}
 deform_pec_base_r {parent=deform_pec_aim_r, loc=<-9.3667 -0.0003 -0.0004>, rot=<0 0 0 1>}
 deform_pec_r {parent=deform_pec_base_r, loc=<0 0 0>, rot=<0 0 0 1>}
 dyn_hoodie_m {parent=spine_04, loc=<0 0 0>, rot=<0 0 0 1>}
 dyn_hoodie_l_master {parent=spine_03, loc=<0.6348 1.2268 -11.7848>, rot=<0.0673 -0.9977 0 0>}
 dyn_hoodie_l {parent=dyn_hoodie_l_master, loc=<0 0 0>, rot=<0 0 0 1>}
 dyn_hoodie_bk_master {parent=spine_03, loc=<2.2258 -6.8408 0>, rot=<0 0 0.9958 -0.0912>}
 dyn_hoodie_bk {parent=dyn_hoodie_bk_master, loc=<0 0 0>, rot=<0 0 0 1>}
 dyn_hoodie_fr_master {parent=spine_03, loc=<-1.8667 12.3099 0>, rot=<0.0799 -0.9968 0 0>}
 dyn_hoodie_fr {parent=dyn_hoodie_fr_master, loc=<0 0 0>, rot=<0 0 0 1>}
 dyn_hoodie_r_master {parent=spine_03, loc=<0.6349 1.2268 11.7848>, rot=<0.0673 -0.9977 0 0>}
 dyn_hoodie_r {parent=dyn_hoodie_r_master, loc=<0 0 0>, rot=<0 0 0 1>}
 thigh_l {parent=pelvis, loc=<-1.0855 -0.2109 -8.7053>, rot=<0.0009 0.0355 0.0155 -0.9993>}
 calf_l {parent=thigh_l, loc=<-42.5284 -0.0019 -0.7297>, rot=<0.0006 -0.0161 -0.0391 -0.9991>}
 calf_twist_01_l {parent=calf_l, loc=<-35.8483 -1.9197 -0.5794>, rot=<0.0237 0.0095 -0.0035 -0.9997>}
 calf_twist_02_l {parent=calf_l, loc=<-19.1546 -1.7943 -0.2638>, rot=<0.0237 0.0095 -0.0035 -0.9997>}
 foot_l {parent=calf_l, loc=<-46.7402 -2.5014 -0.592>, rot=<0.0427 -0.0238 0.0353 -0.9982>}
 ball_l {parent=foot_l, loc=<-3.5658 10.8066 -0.1544>, rot=<0.0001 -0.0001 -0.7063 -0.7079>}
 deform_foot_l_master {parent=foot_l, loc=<2.2282 -3.3321 -0.29>, rot=<-0.0357 -0.7045 -0.0284 -0.7082>}
 deform_foot_l {parent=deform_foot_l_master, loc=<0 0 0>, rot=<0 0 0 1>}
 deform_knee_in_base_l {parent=calf_l, loc=<0.8031 -5.0231 0.2456>, rot=<0 0 0 -1>}
 deform_knee_in_l {parent=deform_knee_in_base_l, loc=<0 0 0>, rot=<0 0 0 1>}
 deform_knee_out_base_l {parent=calf_l, loc=<0 4 0>, rot=<0 0 0 1>}
 deform_knee_out_l {parent=deform_knee_out_base_l, loc=<0 0 0>, rot=<0 0 0 1>}
 thigh_twist_01_l {parent=thigh_l, loc=<-27.952 0 0>, rot=<0.0013 0 0 -1>}
 deform_groin_base_l {parent=thigh_l, loc=<0.4801 5.6776 -0.2486>, rot=<0 0 0 -1>}
 deform_groin_l {parent=deform_groin_base_l, loc=<0 0 0>, rot=<0 0 0 1>}
 thigh_r {parent=pelvis, loc=<-1.0853 -0.2109 8.7053>, rot=<0.0355 -0.0009 0.9993 0.0155>}
 calf_r {parent=thigh_r, loc=<42.5286 0.0019 0.7297>, rot=<0.0006 -0.0161 -0.0391 -0.9991>}
 calf_twist_01_r {parent=calf_r, loc=<35.8485 1.9197 0.5794>, rot=<0.0237 0.0095 -0.0035 -0.9997>}
 calf_twist_02_r {parent=calf_r, loc=<19.1546 1.7943 0.2638>, rot=<0.0237 0.0095 -0.0035 -0.9997>}
 foot_r {parent=calf_r, loc=<46.7403 2.5014 0.592>, rot=<0.0427 -0.0238 0.0353 -0.9982>}
 ball_r {parent=foot_r, loc=<3.5658 -10.8066 0.1544>, rot=<0.0001 -0.0001 -0.7063 -0.7079>}
 deform_foot_r_master {parent=foot_r, loc=<-2.2284 3.3321 0.2898>, rot=<0.7082 0.0284 -0.7045 -0.0357>}
 deform_foot_r {parent=deform_foot_r_master, loc=<0 0 0>, rot=<0 0 0 1>}
 deform_knee_in_base_r {parent=calf_r, loc=<-0.8031 5.0231 -0.2457>, rot=<0 0 0 -1>}
 deform_knee_in_r {parent=deform_knee_in_base_r, loc=<0 0 0>, rot=<0 0 0 1>}
 deform_knee_out_base_r {parent=calf_r, loc=<0 -4 0>, rot=<0 0 0 1>}
 deform_knee_out_r {parent=deform_knee_out_base_r, loc=<0 0 0>, rot=<0 0 0 1>}
 thigh_twist_01_r {parent=thigh_r, loc=<27.952 0 0>, rot=<0.0013 0 0 -1>}
 deform_groin_base_r {parent=thigh_r, loc=<-0.4799 -5.6776 0.2485>, rot=<0 0 0 -1>}
 deform_groin_r {parent=deform_groin_base_r, loc=<0 0 0>, rot=<0 0 0 1>}
 dyn_device_master {parent=thigh_r, loc=<22.6206 -0.3871 7.5754>, rot=<0.0015 0.6816 0.0019 -0.7317>}
 dyn_device_01 {parent=dyn_device_master, loc=<0 0 0>, rot=<0 0 0 1>}
 deform_glute_base_l {parent=pelvis, loc=<-7.3466 -9.0213 -9.1245>, rot=<0.0172 0.031 -0.4477 -0.8935>}
 deform_glute_l {parent=deform_glute_base_l, loc=<0 0 0>, rot=<0 0 0 1>}
 deform_glute_base_r {parent=pelvis, loc=<-7.3464 -9.0213 9.1245>, rot=<0.031 -0.0172 0.8935 -0.4477>}
 deform_glute_r {parent=deform_glute_base_r, loc=<0 0 0>, rot=<0 0 0 1>}
 attach {parent=root, loc=<0 0 0>, rot=<0 0 0 1>}
 ik_foot_root {parent=root, loc=<0 0 0>, rot=<0 0 0 1>}
 ik_foot_l {parent=ik_foot_root, loc=<14.8262 -7.3545 8.296>, rot=<0.0357 0.7045 0.0284 -0.7082>}
 ik_foot_r {parent=ik_foot_root, loc=<-14.8262 -7.3546 8.2958>, rot=<0.7082 0.0284 -0.7045 0.0357>}
 ik_hand_root {parent=root, loc=<0 0 0>, rot=<0 0 0 1>}
 ik_hand_gun {parent=ik_hand_root, loc=<-48.1744 3.3546 104.4381>, rot=<0.6276 0.1309 0.4183 -0.6434>}
 ik_hand_l {parent=ik_hand_gun, loc=<59.3272 67.6997 34.3524>, rot=<-0.6158 -0.7027 -0.3565 0>}
 ik_hand_r {parent=ik_hand_gun, loc=<0 0 0>, rot=<0 0 0 1>}
]
variables [
 Get Delta Ref_Result_Rotation {type=FRotator, default=(Pitch=0,Yaw=0,Roll=0)}
 ItemToModify_3 {
  type=FRigUnit_ModifyTransforms_PerItem
  default=(Item=(Type=Bone,Name="lowerarm_r"),Transform=(Rotation=(X=0,Y=0,Z=0,W=1),Translation=(X=0,Y=0,Z=0),Scale3D=(X=1,Y=1,Z=1)))
 }
 Sprinting {type=bool, default=False}
 isPlayingEmote {type=bool, default=False}
]
